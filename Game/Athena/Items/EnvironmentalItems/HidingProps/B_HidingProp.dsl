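; Base hiding prop Blueprint that players hide inside, with optional teleporting to a linked prop.
; Interacting or landing fast sends gameplay events to pawn abilities, gated by occupancy, burning, blocking tags, angle and line of sight.
; Entry clears obstructions by destroying builds or launching pawns; the first exit can spawn loot, and prop destruction ejects hiders on a timer.
; It also tracks hiders for analytics and disables itself by playlist or ocean submersion.
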(event EventOnMatchStarted
  ;[1 1 1] When match has started and accept navigation update request, set the prop nav indestructible and request navmesh update underneath
  (bind _targetteleporter (Variables|Default|GetTargetTeleporter)
    (:then
      (if (Networking|IsServer)
        (Navigation|SetNavigationIndestructible _targetteleporter true)
        (Navigation|SetNavigationIndestructible true)
        (AI|RequestNavUpdateforBuilding _targetteleporter)
        (AI|RequestNavUpdateforBuilding self)))
    (:else)))

(event EventBeginPlay
  (bind _returnvalue (Actor|GetComponentbyClass "/Script/FortniteGame.FortPhysicsObjectComponent"))
  ;[1 1 1] BeginPlay
  (|CallParentFunction)
  (Utilities|FlowControl|Sequence
    (:then_0
      ;;[1 1 1] Remove fire interaction if it's a teleporting prop
      (Curie|OnCurieActivated
        (:OnCurieActive
          (bind _targetteleporter (Variables|Default|GetTargetTeleporter)
            (:then
              (Curie|DisableElementAttachmentonActor self "(TagName=\"Curie.Element.Fire\")")
              (Curie|DisableElementAttachmentonActor _targetteleporter "(TagName=\"Curie.Element.Fire\")"))
            (:else)))))
    (:then_1
      ;;[1 1 1] Destroy physics component if we're a teleporting prop
      (Variables|Default|GetTargetTeleporter
        (:then
          (Utilities|IsValid _returnvalue
            (:Is%20Valid
              (Components|DestroyComponent _returnvalue))))
        (:else)))
    (:then_2
      (bind _gamestate (Class|FortAsyncActionCurrentPlaylistReady|PlaylistDataReadyAsync)
        (:OnReady
          (Utilities|FlowControl|Sequence
            (:then_0
              (if (GameplayTags|HasTag (Class|FortGameStateAthena|GetAthenaPlaylistContextTags _gamestate) "(TagName=\"Athena.Playlist.Infiltration\")")
                (if (Variables|Default|GetActiveInSpyLTM)
                  _
                  (else
                    _))
                (elif (Variables|Default|GetDestroyInNonSpyLTM)
                  _))
              (Gameplay|SilentDie false))
            (:then_1
              (if (not (Class|AthenaFunctionLibrary|RowtoBool (Variables|Default|GetEnabled)))
                (Gameplay|SilentDie false)))
            (:then_2
              (bind _targetteleporter_2 (Variables|Default|GetTargetTeleporter)
                (:then
                  (CallFunction|NotifyOfBeingATeleporterEndpoint _targetteleporter_2 self)
                  (Variables|Actor|SetbCanBeDamaged false)
                  (if (Networking|IsServer)
                    (Variables|Default|SetIsTeleporter "(TagName=\"Building.Type.Prop.Hiding.Teleporter\")")
                    (if (not (Class|AthenaFunctionLibrary|RowtoBool (Variables|Default|GetTeleportEnabled)))
                      (Gameplay|SilentDie false))))
                (:else)))
            (:then_3
              ;;[1 1 1] If prop is under flood, disable interaction and landed on collision
              (if (Variables|Default|GetDisableWhenSubmergedInWater)
                (bind _asfort_game_state_athena (Utilities|Casting|CastToFortGameStateAthena (Game|GetGameState)))
                (bind _returnvalue_2 (POI|GetPoiGridTagsforLocation _asfort_game_state_athena (Transformation|GetActorLocation self)))
                (if (GameplayTags|HasAnyTags _returnvalue_2 (Variables|Default|GetDisableWhenSubmergedExceptionTags))
                  (Development|PrintString (GameplayTags|GetDebugStringfromGameplayTag (Utilities|Array|Get(acopy) (GameplayTags|BreakGameplayTagContainer _returnvalue_2))))
                  (else
                    (if (Fortnite|WaterHelpers|IsActorSubmergedinOcean self)
                      (Collision|SetCollisionEnabled (Variables|Default|GetLandedOnCollisionMesh) "NoCollision")
                      (Variables|Interaction|SetbAllowInteract false))))))))))))

(event Custom|IgnorePhysicsCollisionDamage
  (params (Target :type Fort%20Pawn%20Object%20Reference) (Ignore%20Duration :type Float%20%28double-precision%29))
  ;[1 1 1] Prevents target from being damaged by this physics object - used when target exits the hiding prop
  (bind _returnvalue (BuildingActor|PhysicsSimulationSupport|GetPhysicsObjectComponent self))
  (Utilities|IsValid _returnvalue
    (:Is%20Valid
      (PhysicsObject|AddtoIgnoredDamageActors _returnvalue Target)
      (Utilities|FlowControl|Delay Ignore%20Duration)
      (PhysicsObject|RemovefromIgnoredDamageActors _returnvalue Target))))

(event Custom|IgnorePawnCollision
  (params (Target :type Fort%20Pawn%20Object%20Reference) (InIgnoreDuration :type Float%20%28double-precision%29 :default 1.0))
  ;[1 1 1] IgnorePawnCollision
  (bind _asfort_player_pawn_athena (Utilities|Casting|CastToFortPlayerPawnAthena Target))
  (Collision|AddIgnoredActor _asfort_player_pawn_athena self InIgnoreDuration))

(event Custom|LaunchPlayersOffTop
  (params (PlayerPawn :type Fort%20Player%20Pawn%20Object%20Reference))
  (bind (_tracestart _traceend _traceextents _traceorientation _actorstoignore) (CallFunction|GetObstructionTraceParams PlayerPawn))
  ;[1 1 1] Launch players off top of hiding prop when player leaves.
  (bind _outhits (CallFunction|GetObstructionObjects (Variables|ObstructionTrace|GetObstructionTraceByObject) _tracestart _traceend _traceextents _traceorientation _actorstoignore))
  (for _array_element _outhits
    (if (Variables|/Script/Engine.Actor|GetbCanBeDamaged (.HitActor (Collision|BreakHitResult _array_element)))
      (Collision|AddIgnoredActor PlayerPawn (.HitActor (Collision|BreakHitResult _array_element)) 0.25) ; Player Exiting Prop ignores Actors it's traced
      ;;[1 1 1] Destroy build pieces and vehicles that are blocking the players exit.
      (bind _asbuilding_actor (Utilities|Casting|CastToBuildingActor (.HitActor (Collision|BreakHitResult _array_element)))
        (:then
          (if (not (GameplayTags|HasMatchingGameplayTag _asbuilding_actor "(TagName=\"Granted.Athena.HidingProp.SmashImmune\")"))
            (Utilities|Casting|CastToBuildingTrap _asbuilding_actor
              (:CastFailed
                (Utilities|Casting|CastToFortPlayerPawn (Actor|GetAttachParentActor _asbuilding_actor)
                  (:CastFailed
                    (if (not (Utilities|Equal(Object) (Actor|GetAttachParentActor _asbuilding_actor) self))
                      (Gameplay|Die _asbuilding_actor "(TagName=\"Gameplay.Effect.InstantDeath\")" (Building|GetOwningController) self)))) ; If attached to a pawn, don't destroy. (like the snowman item)
              ))))
        (:CastFailed
          (Utilities|Casting|CastToFortVehicleInterface (.HitActor (Collision|BreakHitResult _array_element))
            (:then
              (Environment|ApplyGlobalEnvironmentGameplayEffecttoActor (.HitActor (Collision|BreakHitResult _array_element)) "/Game/Athena/Items/EnvironmentalItems/HidingProps/GE_HidingProp_DestroyStructure.GE_HidingProp_DestroyStructure_C" 0 "(GameplayTags=)"))
            (:CastFailed
              (bind _asfort_player_pawn (Utilities|Casting|CastToFortPlayerPawn (.HitActor (Collision|BreakHitResult _array_element))))
              (if (GameplayTags|HasAnyMatchingGameplayTags (Ability|GetFortAbilitySystemComponent _asfort_player_pawn) (GameplayTags|MakeGameplayTagContainerfromArray (Utilities|Array|MakeArray "(TagName=\"Granted.Athena.EnvItem.HidingProp.GettingInProp\")" "(TagName=\"Granted.Athena.EnvItem.HidingProp.HidingInProp\")")))
                (if (GameplayTags|HasAnyMatchingGameplayTags _asfort_player_pawn (GameplayTags|MakeGameplayTagContainerfromArray (Utilities|Array|MakeArray "(TagName=\"Athena.Item.Snowman\")")))
                  _)
                (else
                  _))
              (Character|LaunchCharacter _asfort_player_pawn (+ (Variables|Default|GetAdditionalLaunchVelocity) (select (Variables|Interaction|GetFixedEntranceDirection) (+ (* (Transformation|GetForwardVector (Variables|Default|GetHideLocation_ForwardVector)) 5e+02) "0, 0,750.000000") (Math|Vector|RotateVectorAroundAxis (Math|Vector|MakeVector 5e+02 0.0 7.5e+02) (Math|Random|RandomFloatinRange 0.0 3.6e+02) "0, 0,1.000000"))) true true) ; Launch Player standing on prop in random direction
            ))))
      (else
        (CallFunction|EntranceBlockedbyUndamageable PlayerPawn)))))

(event Custom|ToggleCameraCollisionForClients
  (if (not (Networking|IsServer))
    (for _array_element (Fortnite|GetLocalFortPlayerControllers true true)
      (bind _asfort_pawn (Utilities|Casting|CastToFortPawn (Pawn|GetViewTarget _array_element)))
      (if (CallFunction|IsPawnAlreadyHiding _asfort_pawn)
        (Collision|SetCollisionResponsetoChannel (Variables|Debug|GetStaticMeshComponent) "ECC_Camera" "ECR_Ignore")
        (else
          ;[1 1 1] Band Aid delay fix so that camera doesn't pop as you're exiting.
          (Utilities|Time|SetTimerbyEvent (delegate Custom|TurnClientCameraCollisionOn) (CallFunction|GetTurnClientCameraCollisiononDelayTime)))))))

(event Custom|TurnClientCameraCollisionOn
  ;[1 1 1] Band Aid delay fix so that camera doesn't pop as you're exiting.
  (Collision|SetCollisionResponsetoChannel (Variables|Debug|GetStaticMeshComponent) "ECC_Camera" "ECR_Block"))

(event Custom|EntranceBlockedByUndamageable
  (params (PlayerPawn :type Fort%20Player%20Pawn%20Object%20Reference)))

(event Custom|HidingPlayerCountChanged
  ;[1 1 1] Hidden passage
  (CallFunction|ToggleCameraCollisionforClients)
  (if (Networking|IsServer)
    (if (Utilities|Array|IsNotEmpty (Variables|Default|GetHidingPlayers))
      ;;[1 1 1] Add and Remove Granted tag to the hiding prop if it contains a player.
      (Variables|Default|SetContainsPlayerRepNof "(TagName=\"Granted.Athena.EnvItem.HidingProp.ContainsPlayer\")")
      (else
        (Variables|Default|SetContainsPlayerRepNof)))))

(event Custom|Teleport
  (params (Pawn :type Actor%20Object%20Reference))
  ;[1 1 1] Hidden passage
  (if (Class|AthenaFunctionLibrary|RowtoBool (Variables|Default|GetCanTeleport))
    (if (Networking|IsServer)
      (Variables|Default|GetTargetTeleporter
        (:then
          (bind _returnvalue (GameplayTags|HasMatchingGameplayTag(Message) Pawn "(TagName=\"Gameplay.Status.HidingProp.Teleporting\")"))
          (if (not _returnvalue)
            (GameplayEffects|FortApplyGameplayEffectToTarget :self (Ability|GetFortAbilitySystemComponent self) :GameplayEffectClass (Variables|Default|GetGE_TeleportAbilityGranted) :TargetOfEffect (Ability|GetFortAbilitySystemComponent Pawn) :OptionalSourceObject self :Level 0.0)
            (Ability|SendGameplayEventtoActor Pawn "(TagName=\"Event.Athena.HidingProp.UpdatePropReference\")" (Utilities|Struct|MakeGameplayEventData :Instigator (Variables|Default|GetTargetTeleporter) :OptionalObject (Variables|Default|GetTargetTeleporter) :ContextHandle "()" :InstigatorTags "(GameplayTags=)" :TargetTags "(GameplayTags=)" :EventMagnitude 0.0 :TargetData "()"))))
        (:else)))))

(event Custom|StopHidingLoop
  (bind _int (Variables|Default|GetInt))
  (bind _array (Variables|Default|GetArray))
  (bind _output (Utilities|Array|Get(acopy) _array _int))
  ;[1 1 1] Stop Hiding On Death
  (if (<= (+ (Variables|Default|GetInt) 1) (Utilities|Array|Length (Variables|Default|GetArray)))
    ;[1 1 1] Stop Hiding When Prop is Destroyed Event
    (CallFunction|EndHidingAnalyticSession _output "ObjectDestroyed")
    (Ability|SendGameplayEventtoActor _output "(TagName=\"Event.Athena.HidingProp.StopHiding\")" (Utilities|Struct|MakeGameplayEventData :Instigator self :OptionalObject self :ContextHandle "()" :InstigatorTags "(GameplayTags=)" :TargetTags "(GameplayTags=)" :EventMagnitude 0.2 :TargetData "()"))
    (Math|Integer|IncrementInt (Variables|Default|GetInt)
      (:%20%20
        (Utilities|Time|SetTimerbyEvent (delegate Custom|StopHidingLoop) 0.1 true)))))

(event EventOnDeathServer
  (params (Damage :type Float%20%28single-precision%29) (DamageTags :type Gameplay%20Tag%20Container%20Structure%20%28by%20ref%29) (Momentum :type Vector :default "0, 0, 0") (HitInfo :type Hit%20Result%20Structure%20%28by%20ref%29) (InstigatedBy :type Controller%20Object%20Reference) (DamageCauser :type Actor%20Object%20Reference) (EffectContext :type Gameplay%20Effect%20Context%20Handle%20Structure))
  (bind _int (Variables|Default|GetInt))
  (bind _array (Variables|Default|GetArray))
  (bind _output (Utilities|Array|Get(acopy) _array _int))
  ;[1 1 1] Stop Hiding On Death
  (Variables|Default|SetInt 0)
  (Variables|Default|SetArray (Variables|Default|GetHidingPlayers))
  (if (<= (+ (Variables|Default|GetInt) 1) (Utilities|Array|Length (Variables|Default|GetArray)))
    ;[1 1 1] Stop Hiding When Prop is Destroyed Event
    (CallFunction|EndHidingAnalyticSession _output "ObjectDestroyed")
    (Ability|SendGameplayEventtoActor _output "(TagName=\"Event.Athena.HidingProp.StopHiding\")" (Utilities|Struct|MakeGameplayEventData :Instigator self :OptionalObject self :ContextHandle "()" :InstigatorTags "(GameplayTags=)" :TargetTags "(GameplayTags=)" :EventMagnitude 0.2 :TargetData "()"))
    (Math|Integer|IncrementInt (Variables|Default|GetInt)
      (:%20%20
        (Utilities|Time|SetTimerbyEvent (delegate Custom|StopHidingLoop) 0.1 true)))))

(event OnComponentBeginOverlap(LandedOnCollisionMesh)
  (params (OverlappedComponent :type Primitive%20Component%20Object%20Reference) (OtherActor :type Actor%20Object%20Reference) (OtherComp :type Primitive%20Component%20Object%20Reference) (OtherBodyIndex :type Integer) (bFromSweep :type Boolean) (SweepResult :type Hit%20Result%20Structure%20%28by%20ref%29))
  ;[1 1 1] Land on Prop
  (bind (_canhide _afortplayerpawn) (CallFunction|CanHideByFalling OtherActor))
  (if _canhide
    (CallFunction|LandedonHayStack _afortplayerpawn (.z (Transformation|GetVelocity _afortplayerpawn)))))

(event Custom|AddGameplayCue
  (params (GameplayCueTag :type Gameplay%20Tag%20Structure) (CueParameters :type Gameplay%20Cue%20Parameters%20Structure%20%28by%20ref%29))
  ;[1 1 1] Replicated Gameplay Cue Events
  (if (not (Networking|IsServer))
    (Projectile|AddGameplayCue%28Looping%29Local self GameplayCueTag CueParameters)))

(event Custom|RemoveGameplayCue
  (params (GameplayCueTag :type Gameplay%20Tag%20Structure) (CueParameters :type Gameplay%20Cue%20Parameters%20Structure%20%28by%20ref%29))
  ;[1 1 1] Replicated Gameplay Cue Events
  (if (not (Networking|IsServer))
    (Projectile|RemoveGameplayCue%28Looping%29Local self GameplayCueTag CueParameters)))

(event HidingProp|EventHandleExitPressed
  (params (TargetPlayerPawn :type Fort%20Player%20Pawn%20Object%20Reference))
  ;[1 1 1] Exit with left hand side UI
  (Ability|SendGameplayEventtoActor TargetPlayerPawn "(TagName=\"Event.Athena.HidingProp.StopHiding\")" (Utilities|Struct|MakeGameplayEventData "(TagName=\"Event.Athena.HidingProp.StopHiding\")" self 0 self TargetPlayerPawn "()" "(GameplayTags=)" "(GameplayTags=)" 1.0)))

(event Custom|LandedOnHayStack
  (params (PlayerPawn :type Fort%20Player%20Pawn%20Object%20Reference) (Z%20Velocity%20Mag :type Float%20%28double-precision%29))
  ;[1 1 1] Landed On HidingProp GA Event
  (Ability|SendGameplayEventtoActor PlayerPawn "(TagName=\"Event.Athena.HidingProp.LandedOn\")" (Utilities|Struct|MakeGameplayEventData :Instigator self :OptionalObject self :OptionalObject2 PlayerPawn :ContextHandle "()" :InstigatorTags "(GameplayTags=)" :TargetTags "(GameplayTags=)" :EventMagnitude Z%20Velocity%20Mag :TargetData "()")))

(event Custom|StopHiding
  (params (Pawn :type Fort%20Pawn%20Object%20Reference))
  (bind _lastpawntointeract (Variables|Default|GetLastPawnToInteract))
  ;[1 1 1] ON INTERACT
  ;;[1 1 1] Stop Hiding Custom Event
  (Variables|Default|SetLastPawnToInteract Pawn)
  (CallFunction|EndHidingAnalyticSession _lastpawntointeract "PlayerExit")
  (Ability|SendGameplayEventtoActor _lastpawntointeract "(TagName=\"Event.Athena.HidingProp.StopHiding\")" (Utilities|Struct|MakeGameplayEventData :Instigator self :OptionalObject self :ContextHandle "()" :InstigatorTags "(GameplayTags=)" :TargetTags "(GameplayTags=)" :EventMagnitude 1.0 :TargetData "()"))
  ;[1 1 1] Spawn Loot (once per match)
  (if (and (not (Variables|Default|GetSpawnedLoot)) (Variables|Default|GetShouldSpawnLoot))
    (Variables|Default|SetSpawnedLoot true)
    (bind _itempickupsspawned (Class|AthenaFunctionLibrary|SpawnLootfromTable (Utilities|String|StringToName (Variables|Default|GetLoot%20Tier%20Group)) 1 -1 (+ (Transformation|GetActorLocation (Variables|Default|GetLastPawnToInteract)) (Variables|Default|GetLoot_SpawnOffset)) "0, 0, 0" 0 false false true 0 "FloorLoot" "TossedByPlayer"))
    (CallFunction|LaunchPickups _itempickupsspawned (Variables|Default|GetLastPawnToInteract))))

(event Custom|Launch%20Pickups
  (params (Array :type Array%20of%20Fort%20Pickup%20Object%20References) (Pawn :type Actor%20Object%20Reference))
  ;[1 1 1] Spawn Loot (once per match)
  (Variables|Default|SetLoot_CachedActorForward (Math|Vector|SelectVector (Transformation|GetActorRightVector) (Transformation|GetActorForwardVector Pawn) (Variables|Interaction|GetFixedEntranceDirection)))
  (Variables|Default|SetLoot_VectorOffsets (Utilities|Array|MakeArray))
  (for _ Array
    (Utilities|Array|Add (Variables|Default|GetLoot_VectorOffsets) (Math|Random|RandomUnitVectorinConeinDegrees "0, 0,1.000000" 9e+01)))
  (bind (_direction _z _forward) (Timeline|Loot_MovementTimeline PlayFromStart)
    (:Update
      (for (_array_element _array_index) Array
        (Transformation|SetActorLocation _array_element (Math|Vector|vector+vector (Transformation|GetActorLocation) (Math|Vector|vector*float (Math|Vector|MakeVector 0.0 0.0 (Variables|Default|GetLoot_MoveUpDistance)) _z) (* (Variables|Default|GetLoot_CachedActorForward) (* _forward (Variables|Default|GetLoot_MoveForwardDistance))) (* (Utilities|Array|Get(acopy) (Variables|Default|GetLoot_VectorOffsets) _array_index) (* _forward (Variables|Default|GetLoot_SpawnRadius)))) false true)))))

(event Custom|ExecuteGameplayCue
  (params (GameplayCueTag :type Gameplay%20Tag%20Structure) (CueParameters :type Gameplay%20Cue%20Parameters%20Structure%20%28by%20ref%29))
  ;[1 1 1] Replicated Gameplay Cue Events
  (if (not (Networking|IsServer))
    (GameplayCue|ExecuteGameplayCue%28Burst%29Local self GameplayCueTag CueParameters)))

(event Interaction|EventBlueprintOnInteract
  (params (InteractingPawn :type Fort%20Pawn%20Object%20Reference) (InteractionBeingAttempted :type EInteractionBeingAttempted%20Enum :default "FirstInteraction"))
  ;[1 1 1] ON INTERACT
  (if (and (not (GameplayTags|HasMatchingGameplayTag InteractingPawn "(TagName=\"Weapon.Reload.Failed.ReloadFailedAlreadyFiring\")")) (Interaction|BlueprintCanInteract InteractingPawn InteractionBeingAttempted))
    (Variables|Default|SetLastPawnToInteract InteractingPawn)
    (if (GameplayTags|HasMatchingGameplayTag InteractingPawn "(TagName=\"Granted.Athena.EnvItem.HidingProp.HidingInProp\")")
      (CallFunction|StopHiding (Variables|Default|GetLastPawnToInteract))
      (else
        (CallFunction|StartHiding (Variables|Default|GetLastPawnToInteract))))))

(event Collision|EventActorBeginOverlap
  (params (OtherActor :type Actor%20Object%20Reference))
  ;[1 1 1] While Entering FX
  (bind _asfort_player_pawn (Utilities|Casting|CastToFortPlayerPawn OtherActor))
  (if (GameplayTags|HasMatchingGameplayTag _asfort_player_pawn "(TagName=\"Gameplay.Status.HidingProp.GettingInProp\")")
    (GameplayCue|ExecuteGameplayCue%28Burst%29Local _asfort_player_pawn (Variables|FX|GetWhileEnteringGameplayCue) (Ability|GameplayCue|MakeGameplayCueParameters 0.0))))

(event Custom|StartHiding
  (params (InteractingPawn :type Fort%20Pawn%20Object%20Reference))
  (bind _lastpawntointeract (Variables|Default|GetLastPawnToInteract))
  ;[1 1 1] Start Hiding Event
  ;;[1 1 1] Start Hiding Custom Event
  (Variables|Default|SetLastPawnToInteract InteractingPawn)
  (Ability|SendGameplayEventtoActor (Variables|Default|GetLastPawnToInteract) "(TagName=\"Event.Athena.HidingProp.UpdatePropReference\")" (Utilities|Struct|MakeGameplayEventData :Instigator self :OptionalObject self :ContextHandle "()" :InstigatorTags "(GameplayTags=)" :TargetTags "(GameplayTags=)" :EventMagnitude 0.0 :TargetData "()"))
  (Ability|SendGameplayEventtoActor _lastpawntointeract "(TagName=\"Event.Athena.HidingProp.Hide\")" (Utilities|Struct|MakeGameplayEventData :Instigator self :OptionalObject self :ContextHandle "()" :InstigatorTags "(GameplayTags=)" :TargetTags "(GameplayTags=)" :EventMagnitude 0.0 :TargetData "()"))
  (bind _asfort_player_pawn (Utilities|Casting|CastToFortPlayerPawn _lastpawntointeract))
  (CallFunction|LaunchPlayersOffTop _asfort_player_pawn)
  (CallFunction|InteractEnter))

(event Custom|InteractEnter)

(event Custom|WatchForPlayerDeath
  (params (FortPawn :type Fort%20Pawn%20Object%20Reference))
  ;[1 1 1] Remove Player from Analytics if they die while hiding
  (EventDispatchers|BindEventtoOnDied FortPawn (delegate Custom|Pawn%20Died)))

(event Custom|Pawn%20Died
  (params (DamagedActor :type Actor%20Object%20Reference) (Damage :type Float%20%28single-precision%29) (InstigatedBy :type Controller%20Object%20Reference) (DamageCauser :type Actor%20Object%20Reference) (HitLocation :type Vector) (FHitComponent :type Primitive%20Component%20Object%20Reference) (BoneName :type Name) (Momentum :type Vector))
  ;[1 1 1] Remove Player from Analytics if they die while hiding
  (bind _asfort_player_pawn (Utilities|Casting|CastToFortPlayerPawn DamagedActor))
  (if (Utilities|Map|Contains (Variables|Default|GetHiddenPlayersAndEnterTimes) _asfort_player_pawn)
    (CallFunction|EndHidingAnalyticSession _asfort_player_pawn "PlayerDeath")
    (Utilities|Map|Remove (Variables|Default|GetHiddenPlayersAndEnterTimes) _asfort_player_pawn)
    (PhysicsAnalytics|UpdateHiddenPlayerStat (Variables|Default|GetHiddenPlayersAndEnterTimes))))

(event Custom|HiddenTravelCompleteEvent
  (params (TargetHidingProp :type B%20Hiding%20Prop%20Object%20Reference) (Pawn :type Fort%20Pawn%20Object%20Reference)))

(event Custom|EndHidingAnalyticSession
  (params (FortPawn :type Fort%20Pawn%20Object%20Reference) (EndReason :type EEnvironmentalItemEndReason%20Enum))
  (bind _hiddenplayersandentertimes (Variables|Default|GetHiddenPlayersAndEnterTimes))
  (bind (_value _returnvalue) (Utilities|Map|Find _hiddenplayersandentertimes FortPawn))
  ;[1 1 1] EndHidingAnalyticSession
  (Utilities|IsValid FortPawn
    (:Is%20Valid
      (if _returnvalue
        (bind _asfort_player_controller (Utilities|Casting|CastToFortPlayerController (Pawn|GetController FortPawn)))
        (Analytics|FireAnalyticEventEnvironmentalItemUsed _asfort_player_controller self (- (Utilities|Time|GetRealTimeSeconds) _value) EndReason)))))

(fn AddHiddenPlayer
  (params (FortPawn :type Fort%20Pawn%20Object%20Reference))
  (bind _returnvalue (Ability|GetFortAbilitySystemComponent FortPawn))
  (Variables|Default|SetLastPawnToHide FortPawn)
  (CallFunction|Teleport FortPawn)
  (Utilities|Array|AddUnique (Variables|Default|GetHidingPlayers) FortPawn)
  (Utilities|Map|Add (Variables|Default|GetHiddenPlayersAndEnterTimes) FortPawn (Utilities|Time|GetRealTimeSeconds))
  (PhysicsAnalytics|UpdateHiddenPlayerStat (Variables|Default|GetHiddenPlayersAndEnterTimes))
  (CallFunction|WatchforPlayerDeath FortPawn)
  (CallFunction|IgnorePawnCollision FortPawn)
  ;[1 1 1] apply the prop specific hide effect
  (bind _propspecificeffecttoapplytohiders (Variables|Default|GetPropSpecificEffectToApplyToHiders)
    (:then
      (Utilities|IsValid _returnvalue
        (:Is%20Valid
          (GameplayEffects|FortApplyGameplayEffectToTarget :self (Ability|GetFortAbilitySystemComponent self) :GameplayEffectClass _propspecificeffecttoapplytohiders :TargetOfEffect _returnvalue :OptionalSourceObject self :OptionalEffectCauser self :OptionalDamageSource self :Level 0.0))))
    (:else)))

(fn AllowCosmeticsForPawn
  (params (Pawn :type Fort%20Pawn%20Object%20Reference%20%28by%20ref%29))
  (return (or (not (Utilities|IsValid Pawn :!class /Script/Engine.KismetSystemLibrary)) (not (Utilities|Array|ContainsItem (Variables|Default|GetNonCosmeticPawns) Pawn)))))

(fn BlueprintCanInteract
  (params (InteractingPawn :type Fort%20Pawn%20Object%20Reference) (InteractionBeingAttempted :type EInteractionBeingAttempted%20Enum :default "FirstInteraction") (InteractionType :type TInteractionType%20Enum :default "IT_NoInteraction"))
  (Utilities|IsValid InteractingPawn
    (:Is%20Valid
      ;[1 1 1] Make sure we are allowed to hide before doing anything
      (if (== (ScalableFloat|GetValueatLevel (Variables|Default|GetHidingEnabled)) 1.0)
        ;[1 1 1] Are we a 1-Way Endpoint?
        (if (CallFunction|IsNotOneWayTeleporterEndpoint)
          ;[1 1 1] Do we have enough room?
          (if (not (CallFunction|IsFullyOccupied))
            ;[1 1 1] Disallow interaction if the prop is burning
            (if (not (GameplayTags|HasMatchingGameplayTag "(TagName=\"Curie.State.ElementalStatus.Burning\")"))
              ;[1 1 1] Are we already inside this prop? And not currently getting in that prop?
              (if (and (CallFunction|PlayerHiddenInProp (Variables|GetInteractingPawn)) (not (CallFunction|PlayerGettingInProp (Variables|GetInteractingPawn))))
                ;[1 1 1] If inside the prop, always allow interact, unless we are purposely blocking exit.
                (if (GameplayTags|HasAnyMatchingGameplayTags (Variables|GetInteractingPawn) (GameplayTags|MakeGameplayTagContainerfromArray (Variables|UseConditions|GetBlockExitTags)))
                  (return false)
                  (else
                    (return (CallFunction|ShouldUseInteractionforExit))))
                (elif (not (> (- (.z (Transformation|GetActorLocation)) (.z (Transformation|GetActorLocation (Variables|GetInteractingPawn)))) (Variables|Interaction|GetInteractBelowPropDistance)))
                  ;[1 1 1] If "Only enter from front" is true. Check angle at which player is attempting to enter.
                  (if (or (not (Variables|Interaction|GetFixedEntranceDirection)) (and (Variables|Interaction|GetFixedEntranceDirection) (<= (Math|Trig|Acos(Degrees) (Math|Vector|DotProduct (Transformation|GetForwardVector (Variables|Default|GetHideLocation_ForwardVector)) (Math|Vector|GetForwardVector (Math|Rotator|MakeRotator 0.0 0.0 (.yaw (Math|Rotator|FindLookatRotation (Transformation|GetActorLocation) (Transformation|GetActorLocation (Variables|GetInteractingPawn)))))))) (Variables|Interaction|GetMaxInteractAngle))))
                    ;[1 1 1] Check if player is Not valid for hiding
                    (bind _asfort_player_pawn (Utilities|Casting|CastToFortPlayerPawn (Variables|GetInteractingPawn))
                      (:then
                        (if (Math|Boolean|ORBoolean (Vehicle|IsinVehicle _asfort_player_pawn) (Vehicle|IsPassengerinVehicle _asfort_player_pawn) (GameplayTags|HasAnyMatchingGameplayTags _asfort_player_pawn (GameplayTags|MakeGameplayTagContainerfromArray (Variables|UseConditions|GetBlockEntranceTags))) (AI|Components|NavMovement|IsFalling (Variables|/Script/Engine.Character|GetCharacterMovement _asfort_player_pawn)))
                          (return false)
                          (else
                            _)))
                      (:CastFailed
                        _))
                    ;[1 1 1] Most expensive call, so its last. Check if player has LOS and if so, allow us to interact
                    (bind _in_line_of_sight (Class|MissionBlueprintFunctionLibrary|CheckBasicVectortoActorLineOfSight (Transformation|GetWorldLocation (Variables|Default|GetHideLocation_ForwardVector)) (Variables|GetInteractingPawn) "0, 0, 0" "TraceTypeQuery1" "None" 0.0 (CallFunction|GetIgnoredActorsfromLineOfSight (LocalVariables|GetLOSCheck_IgnoreActors) (Variables|GetInteractingPawn))))
                    (return _in_line_of_sight)
                    (else
                      (return false)))
                  (else
                    (return false))))
              (else
                (return false)))
            (else
              (return false)))
          (else
            (return false)))
        (else
          (return false))))))

(fn BlueprintGetFailedInteractionString
  (params (InteractingPawn :type Fort%20Pawn%20Object%20Reference) (InteractionBeingAttempted :type EInteractionBeingAttempted%20Enum :default "FirstInteraction"))
  (return ""))

(fn BlueprintGetInteractionString
  (params (InteractingPawn :type Fort%20Pawn%20Object%20Reference) (InteractionBeingAttempted :type EInteractionBeingAttempted%20Enum :default "FirstInteraction"))
  (Utilities|IsValid InteractingPawn
    (:Is%20Valid
      (if (CallFunction|IsPawnAlreadyHiding (Variables|GetInteractingPawn))
        (return "NSLOCTEXT(\"[2D91F7A94A7E3856F7DDE1915CBF30C4]\", \"3400A2304D35581C6D00EFB2906AE1FB\", \"Stop Hiding\")")
        (elif (CallFunction|CheckCanTeleportandHaveTarget)
          (return "NSLOCTEXT(\"[2D91F7A94A7E3856F7DDE1915CBF30C4]\", \"0C42618E42C35AEF2382EE92803E6B85\", \"Teleport\")")
          (else
            (return "NSLOCTEXT(\"[2D91F7A94A7E3856F7DDE1915CBF30C4]\", \"9747DE364A84ABE177081CBBD4FA0F2E\", \"Hide\")")))))))

(fn BlueprintGetInteractionTime
  (params (InteractingPawn :type Fort%20Pawn%20Object%20Reference) (InteractionBeingAttempted :type EInteractionBeingAttempted%20Enum :default "FirstInteraction"))
  (return 0.0 true))

(fn BlueprintOnLocalInteract
  (params (InteractingPawn :type Fort%20Player%20Pawn%20Object%20Reference) (InteractionBeingAttempted :type EInteractionBeingAttempted%20Enum :default "FirstInteraction"))
  (Ability|SendGameplayEventtoActor InteractingPawn "(TagName=\"Event.Athena.HidingProp.GettingInProp\")" (Utilities|Struct|MakeGameplayEventData "(TagName=\"Event.Athena.HidingProp.GettingInProp\")" InteractingPawn))
  (return false))

(fn CanHidebyFalling
  (params (OtherActor :type Actor%20Object%20Reference))
  (bind _otheractor (Variables|GetOtherActor))
  (bind _returnvalue (Transformation|GetVelocity _otheractor))
  (bind _z (.z _returnvalue))
  ;[1 1 1] Are we allowed to hide? Are we allowed to hide by falling?
  (if (and (ScalableFloat|GetValueasBool (Variables|Default|GetHidingEnabled)) (ScalableFloat|GetValueasBool (Variables|Default|GetHiding%20Enable%20From%20Landing)))
    ;[1 1 1] Is there room to hide?
    (if (not (CallFunction|IsFullyOccupied))
      ;[1 1 1] Make sure we are not teleporting
      (if (not (Variables|Default|GetTeleporting))
        (if (Networking|IsServer)
          (return false 0)
          (else
            (bind _asfort_player_pawn (Utilities|Casting|CastToFortPlayerPawn (Variables|GetOtherActor))
              (:then
                (LocalVariables|Setlocal_FortPlayerPawn _asfort_player_pawn)
                ;[1 1 1] Not already hiding inside the prop
                (if (not (CallFunction|IsPawnAlreadyHiding (LocalVariables|Getlocal_FortPlayerPawn)))
                  ;[1 1 1] Must be falling fast enough
                  (Development|PrintString (Utilities|String|ToString(Float) _z))
                  (if (<= _z -1.38e+03)
                    ;[1 1 1] Is the pawn allowed to interact with the prop?
                    (if (Fortnite|CanPerformNativeAction (Ability|GetFortAbilitySystemComponent (LocalVariables|Getlocal_FortPlayerPawn)) (GameplayTags|MakeLiteralGameplayTag "(TagName=\"Gameplay.Action.Player.InteractUse\")"))
                      ;[1 1 1] Check for any blocking tags, e.g. DBNO, carrying a player, grinding, etc
                      (if (GameplayTags|HasAnyMatchingGameplayTags (LocalVariables|Getlocal_FortPlayerPawn) (GameplayTags|MakeGameplayTagContainerfromArray (Variables|UseConditions|GetBlockEntranceTags)))
                        (return false 0)
                        (elif (or (Vehicle|IsinVehicle (LocalVariables|Getlocal_FortPlayerPawn)) (Vehicle|IsPassengerinVehicle (LocalVariables|Getlocal_FortPlayerPawn)))
                          ;[1 1 1] Cannot be in a vehicle
                          (return false 0)
                          (else
                            (return true (LocalVariables|Getlocal_FortPlayerPawn)))))
                      (else
                        (return false 0)))
                    (else
                      (return false 0)))
                  (else
                    (return false 0))))
              (:CastFailed
                (return false 0)))))
        (else
          (return false 0)))
      (else
        (return false 0)))
    (else
      (return false 0))))

(fn CheckCanTeleportAndHaveTarget
  (return (Math|Boolean|ANDBoolean (ScalableFloat|GetValueasBool (Variables|Default|GetTeleportEnabled)) (ScalableFloat|GetValueasBool (Variables|Default|GetCanTeleport)) (Utilities|IsValid (Variables|Default|GetTargetTeleporter) :!class /Script/Engine.KismetSystemLibrary))))

(fn CheckCanUsePassage
  (params (Object :type Object%20Reference))
  (bind _asfort_player_pawn_athena (Utilities|Casting|CastToFortPlayerPawnAthena Object)
    (:then
      (if (CallFunction|IsinInfiltrationLTM)
        (return (GameplayTags|HasMatchingGameplayTag :self _asfort_player_pawn_athena))
        (else
          _)))
    (:CastFailed
      _))
  ;[0 0.7448 0.06003] LTM not used, always can teleport
  (return true))

(fn GetObstructionObjects
  (params (TraceByChannel :type Boolean) (Start :type Vector) (End :type Vector) (HalfSize :type Vector) (Orientation :type Rotator) (ActorsToIgnore :type Array%20of%20Actor%20Object%20References))
  (if TraceByChannel
    ;[1 1 1] Used to trace multi objects. Eg. Sewer should break multiple buildings
    (bind _outhits (Collision|MultiBoxTraceForObjects Start End HalfSize Orientation (LocalVariables|GetCollisionObjectTypes) true ActorsToIgnore "None" true "(R=1.000000,G=0.000000,B=0.000000,A=1.000000)" "(R=0.000000,G=1.000000,B=0.000000,A=1.000000)" 3.0))
    (return _outhits)
    (else
      ;[1 1 1] Default usage. Don't get multiple objects at a time.
      (bind _outhits_2 (Collision|MultiBoxTraceByChannel Start End HalfSize Orientation "TraceTypeQuery5" true ActorsToIgnore "None" true "(R=1.000000,G=0.000000,B=0.000000,A=1.000000)" "(R=0.000000,G=1.000000,B=0.000000,A=1.000000)" 3.0))
      (return _outhits_2))))

(fn GetCanRustleAndWobble
  (return (Variables|Default|GetbCanRustleAndWobble)))

(fn GetIgnoredActorsFromLineOfSight
  (params (IgnoredLOSActors :type Array%20of%20Actor%20Object%20References) (InteractingPawn :type Actor%20Object%20Reference))
  ;[0.5625 0.8223 1] Add existing hiding pawns and their attached actors as well
  (bind _output_get (LocalVariables|SetLOSActorsToIgnore (Utilities|Array|MakeArray self))) ; ignore this hiding prop
  (Utilities|IsValid InteractingPawn
    (:Is%20Valid
      (Utilities|Array|AppendArray _output_get IgnoredLOSActors) ; ignore previously set actors\n(often external or overrides)
      (Utilities|Array|AppendArray _output_get (Actor|GetAttachedActors InteractingPawn true false)) ; ignore other actors that are\nattached to interacting pawn
      (Utilities|Array|AppendArray _output_get (Utilities|Array|MakeArray (Actor|GetAttachParentActor InteractingPawn))) ; ignore actors the pawn \nis attached to - may handle blocking \ninteraction other ways
    )
    (:Is%20Not%20Valid
      _))
  (for _array_element (Variables|Default|GetHidingPlayers)
    (Utilities|Array|Add _output_get _array_element) ; add all hiding pawns
    (Utilities|Array|AppendArray _output_get (Actor|GetAttachedActors _array_element true false)) ; ...and anything attached to them
  )
  ;[1 0.4145 0.9886] Return - Actors to ignore during LOS check
  (return (LocalVariables|GetLOSActorsToIgnore)))

(fn GetMarkerDisplayData
  (return (Variables|Marking|GetMarkerDisplay)))

(fn GetMarkerPositionOffset
  (return (Variables|Marking|GetMarkerPositionOffset)))

(fn GetMinDistanceFromInteraction
  (return 0.0))

(fn GetObstructionTraceParams
  (params (InCharacter :type Character%20Object%20Reference))
  (return (+ (Transformation|GetWorldLocation (Variables|Default|GetHideLocation_ForwardVector)) (Variables|ObstructionTrace|GetObstructionTraceStartOffSet)) (+ (Transformation|GetWorldLocation (Variables|Default|GetHideLocation_ForwardVector)) (* (Math|Vector|SelectVector (+ (Variables|ObstructionTrace|GetFixedEntraceObstructionTraceEndOffset) (Transformation|GetForwardVector (Variables|Default|GetHideLocation_ForwardVector))) (Variables|ObstructionTrace|GetObstruction%20Trace%20End) (Variables|Interaction|GetFixedEntranceDirection)) (Variables|ObstructionTrace|GetObstructionTraceLength))) (Variables|ObstructionTrace|GetObstructionTraceExtents) (Math|Conversions|RotationFromXVector (Transformation|GetForwardVector (Variables|Default|GetHideLocation_ForwardVector))) (Utilities|Array|MakeArray (Variables|GetInCharacter))))

(fn GetPlayerLimit
  (return (Math|Float|Truncate (ScalableFloat|GetValueatLevel (Variables|Default|GetPlayerLimit)))))

(fn GetTurnClientCameraCollisionOnDelayTime
  (return (Animation|GetPlayLength (Variables|Animation|GetExitAnimMontage))))

(fn IsAOneWayTeleporterEndpoint
  ;[1 1 1] Are we a teleporter, if so we cant be a 1-Way Endpoint
  (Variables|Default|GetTargetTeleporter
    (:then
      (return false))
    (:else
      ;[1 1 1] If we are not a teleporter but we are referenced by teleporters them we are a 1-Way Endpoint
      (return (Utilities|Array|IsNotEmpty (Variables|Default|GetTeleportersTargetingMe))))))

(fn IsFullyOccupied
  (if (Variables|Default|GetSingleOccupant)
    (return (Utilities|Array|IsNotEmpty (Variables|Default|GetHidingPlayers)))
    (else
      (return (>= (Utilities|Array|Length (Variables|Default|GetHidingPlayers)) (CallFunction|GetPlayerLimit))))))

(fn IsInInfiltrationLTM
  (bind _asfort_game_state_athena (Utilities|Casting|CastToFortGameStateAthena (Game|GetGameState))
    (:then
      (return (GameplayTags|HasTag (Class|FortGameStateAthena|GetAthenaPlaylistContextTags _asfort_game_state_athena) "(TagName=\"Athena.Playlist.Infiltration\")")))
    (:CastFailed
      ;[1 0.2076 0] This returns GameplayTags from the Playlist as well as any other "context" tags, to the type of the LMT; solo, squads, etc...
      (return false))))

(fn IsNotOneWayTeleporterEndpoint
  (return (not (CallFunction|IsAOneWayTeleporterEndpoint))))

(fn IsPawnAlreadyHiding
  (params (Pawn :type Fort%20Pawn%20Object%20Reference))
  (return (Utilities|Array|ContainsItem (Variables|Default|GetHidingPlayers) Pawn)))

(fn LocalOnFailedInteract
  (params (InteractingPawn :type Fort%20Player%20Pawn%20Object%20Reference)))

(fn NotifyOfBeingATeleporterEndpoint
  (params (Teleporter :type B%20Hiding%20Prop%20Object%20Reference))
  (bind _teleporter (Variables|GetTeleporter)
    (:then
      (Utilities|Array|AddUnique (Variables|Default|GetTeleportersTargetingMe) (Utilities|ResolveSoftReference _teleporter))
      ;[1 1 1] If we are a teleporter target, we should be indestructible
      (if (Utilities|Array|IsNotEmpty (Variables|Default|GetTeleportersTargetingMe))
        (Variables|Actor|SetbCanBeDamaged false)))
    (:else)))

(fn OnRep_ContainsPlayer
  (bind _returnvalue (GameplayTags|MakeGameplayTagContainerfromTag "(TagName=\"Granted.Athena.EnvItem.HidingProp.ContainsPlayer\")"))
  (bind _self self)
  (if (GameplayTags|IsGameplayTagValid (Variables|Default|GetContainsPlayerRepNof))
    (Ability|GameplayEffect|AddLooseGameplayTags _self _returnvalue)
    (else
      (Ability|GameplayEffect|RemoveLooseGameplayTags _self _returnvalue))))

(fn OnRep_HidingPlayers
  (Development|PrintString (Utilities|String|ToString(Integer) (Utilities|Array|Length (Variables|Default|GetHidingPlayers)))))

(fn OnRep_IsTeleporter
  (Ability|GameplayEffect|AddLooseGameplayTags self (GameplayTags|MakeGameplayTagContainerfromTag (Variables|Default|GetIsTeleporter))))

(fn OnRep_Teleporting)

(fn PlayerGettinginProp
  (params (FortPawn :type Fort%20Pawn%20Object%20Reference))
  (return (GameplayTags|HasMatchingGameplayTag FortPawn "(TagName=\"Granted.Athena.EnvItem.HidingProp.GettingInProp\")")))

(fn PlayerHiddeninProp
  (params (Fort%20Pawn :type Fort%20Pawn%20Object%20Reference))
  (return (and (CallFunction|IsPawnAlreadyHiding Fort%20Pawn) (not (GameplayTags|HasMatchingGameplayTag Fort%20Pawn "(TagName=\"Granted.Athena.EnvItem.HidingProp.GettingInProp\")")))))

(fn RemoveHiddenPlayer
  (params (FortPawn :type Fort%20Pawn%20Object%20Reference))
  (bind _hiddenplayersandentertimes (Variables|Default|GetHiddenPlayersAndEnterTimes))
  (Utilities|Array|RemoveItem (Variables|Default|GetHidingPlayers) FortPawn)
  (Utilities|Map|Remove _hiddenplayersandentertimes FortPawn)
  (PhysicsAnalytics|UpdateHiddenPlayerStat _hiddenplayersandentertimes)
  (Development|PrintString (Utilities|String|Append "Removing Hidden Player:" (Utilities|GetDisplayName FortPawn)))
  ;[1 1 1] remove prop specific hide effect
  (bind _propspecificeffecttoapplytohiders (Variables|Default|GetPropSpecificEffectToApplyToHiders)
    (:then
      (GameplayEffects|RemoveActiveGameplayEffectbySourceEffect (Ability|GetAbilitySystemComponent (Variables|GetFortPawn)) _propspecificeffecttoapplytohiders 0 -1))
    (:else)))

(fn ConstructionScript)
