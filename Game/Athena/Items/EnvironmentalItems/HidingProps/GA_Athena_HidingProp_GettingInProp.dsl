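; Gameplay ability for the Athena hiding prop that covers the short window while a player is getting into a prop.
; On activation it starts a gameplay tag query task that matches any of the getting-in, hiding-in-prop or hide-player-pawn tags.
; It ends when that query triggers, with a parallel one-second delay as a timeout that also ends it.
; The end-ability event has no logic, and the graph applies no effects or tags itself.
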(event Ability|EventActivateAbility
  (Ability|Tasks|WaitGameplayTagQuery "(TokenStreamVersion=0,TagDictionary=((TagName=\"Granted.Athena.EnvItem.HidingProp.GettingInProp\"),(TagName=\"Granted.Athena.EnvItem.HidingProp.HidingInProp\"),(TagName=\"Pawn.HidePlayerPawn.HidingProp\")),QueryTokenStream=(0,1,1,3,0,1,2),UserDescription=\"StartedHiding\",AutoDescription=\" ANY( Granted.Athena.EnvItem.HidingProp.GettingInProp, Granted.Athena.EnvItem.HidingProp.HidingInProp, Pawn.HidePlayerPawn.HidingProp )\")"
    (:then
      (Utilities|FlowControl|Delay 1.0)
      (Ability|EndAbility))
    (:Triggered
      (Ability|EndAbility))))

(event Ability|EventOnEndAbility
  (params (bWasCancelled :type Boolean)))
