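; Gameplay ability that handles a player hiding in, landing on and leaving an environmental hiding prop.
; The triggering event tag selects the flow: montage and root-motion entry, instant land-on/teleport entry, or launch or walk-out exit.
; Hiding toggles pawn visibility, collision, movement, camera mode and status effects, with next-tick position locking on physics props.
; A loop plays rustle cues and wobbles, while server checks watch for DBNO, carrying or snowman states.
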(event Custom|RustleLoop
  (bind _hidingprop (Variables|Default|GetHidingProp))
  (bind _rustlesperwobble (Variables|/Game/Athena/Items/EnvironmentalItems/HidingProps/B_HidingProp.B_HidingProp_C|GetRustlesPerWobble _hidingprop))
  (bind _returnvalue (ScalableFloat|GetValueatLevel _rustlesperwobble))
  (bind _returnvalue_2 (Math|Float|Truncate _returnvalue))
  ;[1 1 1] Make Occasional rustle sound
  (if (Variables|/Game/Athena/Items/EnvironmentalItems/HidingProps/B_HidingProp.B_HidingProp_C|GetbCanRustleAndWobble (Variables|Default|GetHidingProp))
    (Utilities|FlowControl|Sequence
      (:then_0
        _)
      (:then_1
        _))
    (Utilities|FlowControl|Gate true
      (:Exit
        (Utilities|FlowControl|Delay (Math|Random|RandomFloatinRange (ScalableFloat|GetValueatLevel (Variables|Default|GetRustleMinDelay)) (ScalableFloat|GetValueatLevel (Variables|Default|GetRustleMaxDelay))))
        (Ability|ExecuteGameplayCueWithParamsOnOwner (Variables|/Game/Athena/Items/EnvironmentalItems/HidingProps/B_HidingProp.B_HidingProp_C|GetRustleGameplayCue (Variables|Default|GetHidingProp)) (Ability|GameplayCue|MakeGameplayCueParameters :NormalizedMagnitude 0.0 :RawMagnitude 0.0 :Location "0,0,0" :Normal "0,0,0" :Instigator (Variables|Default|GetHidingProp) :GameplayEffectLevel 1 :AbilityLevel 1 :bReplicateLocationWhenUsingMinimalRepProxy false))
        ;;[1 1 1] Wobble every few rustles
        (Utilities|FlowControl|Sequence
          (:then_0
            _)
          (:then_1
            _))
        (bind _counter (Utilities|FlowControl|DoN _returnvalue_2)
          (:Exit
            (if (>= _counter _returnvalue_2)
              (Utilities|FlowControl|Sequence
                (:then_0
                  _)
                (:then_1
                  _))
              (CallFunction|WobbleHidingProp (ScalableFloat|GetValueatLevel (Variables|/Game/Athena/Items/EnvironmentalItems/HidingProps/B_HidingProp.B_HidingProp_C|GetRustleWobbleRadius (Variables|Default|GetHidingProp)))))))))))

(event Ability|EventOnEndAbility
  (params (bWasCancelled :type Boolean))
  (bind _hidingprop (Variables|Default|GetHidingProp))
  (bind _rustlesperwobble (Variables|/Game/Athena/Items/EnvironmentalItems/HidingProps/B_HidingProp.B_HidingProp_C|GetRustlesPerWobble _hidingprop))
  (bind _returnvalue (ScalableFloat|GetValueatLevel _rustlesperwobble))
  (bind _returnvalue_2 (Math|Float|Truncate _returnvalue))
  ;[1 1 1] Make Occasional rustle sound
  ;;[1 1 1] Remove Getting In Prop GE if we have it.
  (Ability|RemoveGameplayEffectFromOwnerWithGrantedTags "(GameplayTags=((TagName=\"Gameplay.Status.HidingProp.GettingInProp\")))")
  (Rendering|SetActorHiddenInGameWithTag (Variables|Default|GetPlayerPawn) false "(TagName=\"Pawn.HidePlayerPawn.HidingProp\")" (Variables|Default|GetHidingProp) true)
  (CallFunction|SetCosmeticsAllowed true)
  (Ability|SendGameplayEventtoActor (Variables|Default|GetPlayerPawn) "(TagName=\"Event.Athena.HidingProp.PropExited\")" (Utilities|Struct|MakeGameplayEventData :Instigator (Variables|Default|GetHidingProp) :OptionalObject (Variables|Default|GetHidingProp) :ContextHandle "()" :InstigatorTags "(GameplayTags=)" :TargetTags "(GameplayTags=)" :EventMagnitude 0.0 :TargetData "()"))
  (Default|DebugAbilityLogging true false "GA_Athena_HidingProp_Hide: OnEndAbility.
SendGameplayEventToActor: Event.Athena.HidingProp.PropExited")
  ;;[1 0.7719 0] Camera Mod Removed
  (Camera|PopCameraMode)
  (if (not (Networking|IsServer))
    (bind _hidingprop_2 (Variables|Default|GetHidingProp)
      (:then
        (HidingProp|SetInputEnabled _hidingprop_2 (Variables|Default|GetPlayerPawn) false))
      (:else
        _))
    (Utilities|FlowControl|Gate true
      (:Exit
        (Utilities|FlowControl|Delay (Math|Random|RandomFloatinRange (ScalableFloat|GetValueatLevel (Variables|Default|GetRustleMinDelay)) (ScalableFloat|GetValueatLevel (Variables|Default|GetRustleMaxDelay))))
        (Ability|ExecuteGameplayCueWithParamsOnOwner (Variables|/Game/Athena/Items/EnvironmentalItems/HidingProps/B_HidingProp.B_HidingProp_C|GetRustleGameplayCue (Variables|Default|GetHidingProp)) (Ability|GameplayCue|MakeGameplayCueParameters :NormalizedMagnitude 0.0 :RawMagnitude 0.0 :Location "0,0,0" :Normal "0,0,0" :Instigator (Variables|Default|GetHidingProp) :GameplayEffectLevel 1 :AbilityLevel 1 :bReplicateLocationWhenUsingMinimalRepProxy false))
        ;;[1 1 1] Wobble every few rustles
        (Utilities|FlowControl|Sequence
          (:then_0
            _)
          (:then_1
            _))
        (bind _counter (Utilities|FlowControl|DoN _returnvalue_2)
          (:Exit
            (if (>= _counter _returnvalue_2)
              (Utilities|FlowControl|Sequence
                (:then_0
                  _)
                (:then_1
                  _))
              (CallFunction|WobbleHidingProp (ScalableFloat|GetValueatLevel (Variables|/Game/Athena/Items/EnvironmentalItems/HidingProps/B_HidingProp.B_HidingProp_C|GetRustleWobbleRadius (Variables|Default|GetHidingProp)))))))))))

(event Custom|Event_LockOnProp
  ;[1 1 1] Attach Player to Hiding Prop
  (Utilities|Time|ClearandInvalidateTimerbyHandle (Variables|Default|GetLockOnTimer))
  (Transformation|SetActorLocation (Variables|Default|GetPlayerPawn) (Transformation|GetWorldLocation (Variables|/Game/Athena/Items/EnvironmentalItems/HidingProps/B_HidingProp.B_HidingProp_C|GetHideLocation_ForwardVector (Variables|Default|GetHidingProp))) false false)
  (bind _returnvalue (Utilities|Time|SetTimerforNextTickbyEvent (delegate Custom|Event_LockOnProp)))
  (Variables|Default|SetLockOnTimer _returnvalue))

(event Ability|EventActivateAbilityFromEvent
  (params (EventData :type Gameplay%20Event%20Data%20Structure%20%28by%20ref%29))
  (bind (_eventtag _instigator _target _optionalobject _optionalobject2 _contexthandle _instigatortags _targettags _eventmagnitude) (Utilities|Struct|BreakGameplayEventData EventData))
  (bind _hidingprop (Variables|Default|GetHidingProp))
  (bind _movetoactordelayonenter (Variables|/Game/Athena/Items/EnvironmentalItems/HidingProps/B_HidingProp.B_HidingProp_C|GetMoveToActorDelayOnEnter _hidingprop))
  (bind _returnvalue (Math|Float|Max(Float) _movetoactordelayonenter))
  (bind _returnvalue_2 (- (Animation|GetPlayLength (Variables|Animation|GetEnter%20Anim%20Montage)) _returnvalue))
  (bind _returnvalue_3 (- _returnvalue_2 0.3))
  (bind _hidingprop_2 (Variables|Default|GetHidingProp))
  (bind _movetoactordurationoverride (Variables|/Game/Athena/Items/EnvironmentalItems/HidingProps/B_HidingProp.B_HidingProp_C|GetMoveToActorDurationOverride _hidingprop_2))
  (bind _returnvalue_4 (< _movetoactordurationoverride 0))
  (bind _returnvalue_5 (Math|Float|SelectFloat _returnvalue_3 _movetoactordurationoverride _returnvalue_4))
  (bind _enter_anim_montage (Variables|Animation|GetEnter%20Anim%20Montage))
  (bind _returnvalue_6 (Animation|GetPlayLength _enter_anim_montage))
  (bind _returnvalue_7 (== _eventmagnitude 0.2))
  (bind _returnvalue_8 (Math|Float|Max(Float) _returnvalue_5 0.2))
  (bind _hidingprop_3 (Variables|Default|GetHidingProp))
  (bind _exitanimmontage (Variables|/Game/Athena/Items/EnvironmentalItems/HidingProps/B_HidingProp.B_HidingProp_C|GetExitAnimMontage _hidingprop_3))
  (bind _playerpawn (Variables|Default|GetPlayerPawn))
  (bind _hidingprop_4 (Variables|Default|GetHidingProp))
  (bind _charactermovement (Variables|/Script/Engine.Character|GetCharacterMovement _playerpawn))
  (bind _hidingprop_5 (Variables|Default|GetHidingProp))
  (bind _playerpawn_2 (Variables|Default|GetPlayerPawn))
  (bind _hidingprop_6 (Variables|Default|GetHidingProp))
  (bind _previnputenabledhidingprop (Variables|Default|GetPrevInputEnabledHidingProp))
  (bind _playerpawn_3 (Variables|Default|GetPlayerPawn))
  (bind _charactermovement_2 (Variables|/Script/Engine.Character|GetCharacterMovement _playerpawn_3))
  (bind _playerpawn_4 (Variables|Default|GetPlayerPawn))
  (bind _playerpawn_5 (Variables|Default|GetPlayerPawn))
  ;[1 1 1] Save Pawn, Controller, And Hiding Prop
  (bind _asb_hiding_prop (Utilities|Casting|CastToB_HidingProp _instigator)
    (:then
      (bind _output_get (Variables|Default|SetHidingProp _asb_hiding_prop))
      (Variables|Default|SetHideLocationForwardVectorWorldPos (Transformation|GetWorldLocation (Variables|/Game/Athena/Items/EnvironmentalItems/HidingProps/B_HidingProp.B_HidingProp_C|GetHideLocation_ForwardVector _output_get)))
      (Variables|Default|SetHideLocationForwardVectorWorldRot (Transformation|GetWorldRotation (Variables|/Game/Athena/Items/EnvironmentalItems/HidingProps/B_HidingProp.B_HidingProp_C|GetHideLocation_ForwardVector _output_get))))
    (:CastFailed
      _))
  (bind _asfort_player_pawn (Utilities|Casting|CastToFortPlayerPawn (Ability|GetAvatarActorfromActorInfo))
    (:then
      (bind _output_get_2 (Variables|Default|SetPlayerPawn _asfort_player_pawn))
      (Utilities|IsValid (Fortnite|GetFortPlayerControllerfromActor _output_get_2)
        (:Is%20Valid
          (Variables|Default|SetPlayerController (Fortnite|GetFortPlayerControllerfromActor _output_get_2))
          (CallFunction|SetCosmeticsAllowed (not (GameplayTags|HasTag _instigatortags (Variables|Default|GetDisallowCosmeticsTag))))
          ;[1 1 1] Switch On Event TAG
          (switch Utilities|FlowControl|Switch|SwitchonGameplayTag _eventtag
            (:Event.Athena.HidingProp.Hide
              (bind _asb_hiding_prop_2 (Utilities|Casting|CastToB_HidingProp _instigator)
                (:then
                  (Variables|Default|SetHidingProp _asb_hiding_prop_2)
                  (Ability|SendGameplayEventtoActor (Variables|Default|GetPlayerPawn) "(TagName=\"Event.Athena.HidingProp.Interact\")" (Utilities|Struct|MakeGameplayEventData :Instigator (Variables|Default|GetHidingProp) :OptionalObject (Variables|Default|GetHidingProp) :ContextHandle "()" :InstigatorTags "(GameplayTags=)" :TargetTags "(GameplayTags=)" :EventMagnitude 0.0 :TargetData "()"))
                  (Utilities|FlowControl|Sequence
                    (:then_1
                      (Ability|ApplyGameplayEffectToOwner "/Game/Athena/Items/EnvironmentalItems/HidingProps/GE_Athena_HidingProp_GettingInProp.GE_Athena_HidingProp_GettingInProp_C")
                      ;[1 1 1] Update The Hiding Prop of our Entrance
                      (Class|BHidingProp|AddHiddenPlayer (Variables|Default|GetHidingProp) (Variables|Default|GetPlayerPawn))
                      (Class|BHidingProp|HidingPlayerCountChanged (Variables|Default|GetHidingProp))
                      ;[1 1 1] Move Player into Prop over duration of Animation.
                      (Animation|StopAnimMontage (Variables|Default|GetPlayerPawn) (Animation|GetCurrentMontage (Variables|Default|GetPlayerPawn) :!class /Script/Engine.Character))
                      (Ability|Tasks|PlayMontageandWaitforTrigger "None" (Variables|/Game/Athena/Items/EnvironmentalItems/HidingProps/B_HidingProp.B_HidingProp_C|GetEnterAnimMontage (Variables|Default|GetHidingProp)))
                      (Utilities|FlowControl|Sequence
                        (:then_0
                          (if (> (Variables|/Game/Athena/Items/EnvironmentalItems/HidingProps/B_HidingProp.B_HidingProp_C|GetMoveToActorDelayOnEnter (Variables|Default|GetHidingProp)) 0)
                            (Utilities|FlowControl|Delay (Variables|/Game/Athena/Items/EnvironmentalItems/HidingProps/B_HidingProp.B_HidingProp_C|GetMoveToActorDelayOnEnter (Variables|Default|GetHidingProp)))
                            (else
                              _))
                          (Utilities|IsValid (Variables|Default|GetPlayerPawn)
                            (:Is%20Valid
                              (if (Variables|/Game/Athena/Items/EnvironmentalItems/HidingProps/B_HidingProp.B_HidingProp_C|GetBypassMoveToActor (Variables|Default|GetHidingProp))
                                (Utilities|FlowControl|Sequence
                                  (:then_0
                                    _)
                                  (:then_1
                                    ;;[1 1 1] Is In Hiding Prop
                                    (Ability|ApplyGameplayEffectToOwner "/Game/Athena/Items/EnvironmentalItems/HidingProps/GE_Athena_HidingProp_HidingInProp.GE_Athena_HidingProp_HidingInProp_C")
                                    (Utilities|FlowControl|Sequence
                                      (:then_0
                                        (Variables|Default|SetLandedOnProp false)
                                        ;[1 1 1] Set to None So Gravity Isn't Applied and so the player can't be impulsed or launched.
                                        (Components|Movement|StopMovementImmediately _charactermovement)
                                        (Pawn|Components|CharacterMovement|SetMovementMode _charactermovement "MOVE_None" 0)
                                        (if (Variables|/Game/Athena/Items/EnvironmentalItems/HidingProps/B_HidingProp.B_HidingProp_C|GetSetEntranceRotation (Variables|Default|GetHidingProp))
                                          ;[0.9923 0.9439 1] Force Set Location and Rotation
                                          (Transformation|SetActorLocationAndRotation (Variables|Default|GetPlayerPawn) (+ (Variables|Default|GetHideLocationForwardVectorWorldPos) "0, 0,10.000000") (Math|Rotator|CombineRotators (Variables|Default|GetHideLocationForwardVectorWorldRot) (Variables|/Game/Athena/Items/EnvironmentalItems/HidingProps/B_HidingProp.B_HidingProp_C|GetAddedSetEntranceRotation (Variables|Default|GetHidingProp))) false true)
                                          (else
                                            ;[1 1 1] Force Set Location
                                            (Transformation|SetActorLocation (Variables|Default|GetPlayerPawn) (+ (Variables|Default|GetHideLocationForwardVectorWorldPos) "0, 0,10.000000") false true)))
                                        ;[1 1 1] Hide the player
                                        (Rendering|SetActorHiddenInGameWithTag (Variables|Default|GetPlayerPawn) true "(TagName=\"Pawn.HidePlayerPawn.HidingProp\")" (Variables|Default|GetHidingProp) true)
                                        (if (Variables|/Game/Athena/Items/EnvironmentalItems/HidingProps/B_HidingProp.B_HidingProp_C|GetIgnorePhysicsBodyCollisionOnEnter (Variables|Default|GetHidingProp))
                                          (Collision|SetCollisionResponsetoChannel (Variables|/Script/Engine.Character|GetCapsuleComponent (Variables|Default|GetPlayerPawn)) "ECC_PhysicsBody" "ECR_Block")
                                          (else
                                            _))
                                        (Collision|SetActorEnableCollision (Variables|Default|GetPlayerPawn) false) ; Prevent issues with the pawn retaining collision while using\nhiding props that are smaller than them (like the sewer tunnel)
                                        ;[1 1 1] Attach Player to Hiding Prop
                                        (bind _returnvalue_9 (FortPhysicsSimulationLibrary|CanActorSimulatePhysics (Variables|Default|GetHidingProp)))
                                        (if _returnvalue_9
                                          (bind _returnvalue_10 (Utilities|Time|SetTimerforNextTickbyEvent (delegate Custom|Event_LockOnProp)))
                                          (Variables|Default|SetLockOnTimer _returnvalue_10)
                                          (else
                                            _))
                                        (if (Networking|IsServer)
                                          ;[1 1 1] Skip entrance cues if using Prop as teleporter or AllowCosmetics is false
                                          (if (not (or (Variables|/Game/Athena/Items/EnvironmentalItems/HidingProps/B_HidingProp.B_HidingProp_C|GetTeleporting (Variables|Default|GetHidingProp)) (not (Variables|Default|GetAllowCosmetics))))
                                            (Ability|ExecuteGameplayCueWithParamsOnOwner (Variables|/Game/Athena/Items/EnvironmentalItems/HidingProps/B_HidingProp.B_HidingProp_C|GetEnterGameplayCue (Variables|Default|GetHidingProp)) (Ability|GameplayCue|MakeGameplayCueParameters :NormalizedMagnitude 0.0 :RawMagnitude 0.0 :Location "0,0,0" :Normal "0,0,0" :Instigator (Variables|Default|GetHidingProp) :GameplayEffectLevel 1 :AbilityLevel 1 :bReplicateLocationWhenUsingMinimalRepProxy false))
                                            (CallFunction|WobbleHidingProp (ScalableFloat|GetValueatLevel (Variables|/Game/Athena/Items/EnvironmentalItems/HidingProps/B_HidingProp.B_HidingProp_C|GetEnterWobbleRadius (Variables|Default|GetHidingProp)))))
                                          (else
                                            ;;[1 1 1] Store HidingProp w/ Enabled Input in case of Teleport.
                                            ;; Disable When Teleport succeeds (OnLanded)
                                            (HidingProp|SetInputEnabled _hidingprop_5 (Variables|Default|GetPlayerPawn) true)
                                            (Variables|Default|SetPrevInputEnabledHidingProp _hidingprop_5) ; In case of a desync with the server TargetTeleporter validity. \nAlways let the player exit hiding. If the teleport succeeds CanInteract handles keeping the player from leaving the prop. \nIf the server desyncs and cancels a teleport the client thinks will be valid then let the player leave the hiding spot.
                                          ))
                                        ;[1 1 1] Send quest progress update
                                        (bind _output_get_3 (Variables|Quests|SetTC_PlayerOwned (GameplayTags|GetOwnedGameplayTags _playerpawn_4 :!class /Script/GameplayTags.GameplayTagAssetInterface)))
                                        (GameplayTags|AppendGameplayTagContainers _output_get_3 (Variables|Quests|GetTC_Quest))
                                        (Quest|SendComplexCustomStatEvent (Quests|GetQuestManager (Fortnite|GetFortPlayerControllerfromActor _playerpawn_4) "Invalid") (Variables|Default|GetHidingProp) (Variables|Quests|GetTC_PlayerOwned) (GameplayTags|GetOwnedGameplayTags (Variables|Default|GetHidingProp) :!class /Script/GameplayTags.GameplayTagAssetInterface) 1)
                                        (CallFunction|RustleLoop)
                                        ;[0.9902 1 0.4405] Final check incase we've missed the death on server event from the prop, or if the player started carrying a player or died and is being carried at the same time the "land on prop" overlap event is called. if any of these are true, stop hiding. NOTE: This should be converted to Async Wait tasks that are always listening for these tags to be added while this ability is running
                                        (Utilities|IsServer
                                          (:true
                                            (if (Math|Boolean|ORBoolean (Variables|/Script/FortniteGame.BuildingActor|GetbDestroyed (Variables|Default|GetHidingProp)) (GameplayTags|HasAnyMatchingGameplayTags (Variables|Default|GetPlayerPawn) (GameplayTags|MakeGameplayTagContainerfromArray (Utilities|Array|MakeArray "(TagName=\"Gameplay.Action.Player.CarryingPlayer\")" "(TagName=\"Athena.Item.Snowman\")" "(TagName=\"Gameplay.Status.DBNO\")"))) (Vehicle|IsinVehicle (Variables|Default|GetPlayerPawn)) (Vehicle|IsPassengerinVehicle (Variables|Default|GetPlayerPawn)))
                                              (Ability|SendGameplayEvent "(TagName=\"Event.Athena.HidingProp.StopHiding\")" (Utilities|Struct|MakeGameplayEventData :ContextHandle "()" :InstigatorTags "(GameplayTags=)" :TargetTags "(GameplayTags=)" :EventMagnitude 0.8 :TargetData "()"))
                                              (elif (Variables|Default|GetLandedOnProp)
                                                _
                                                (else
                                                  (Networking|ForceNetUpdate (Variables|Default|GetPlayerPawn))
                                                  (Utilities|FlowControl|Sequence
                                                    (:then_0
                                                      ;[1 1 1] Remove getting in prop GE
                                                      (Utilities|FlowControl|Delay 1.0)
                                                      (Ability|RemoveGameplayEffectFromOwnerWithGrantedTags "(GameplayTags=((TagName=\"Granted.Athena.EnvItem.HidingProp.GettingInProp\")))"))
                                                    (:then_1
                                                      (Utilities|FlowControl|Delay 0.2)
                                                      ;[1 1 1] Hide the player
                                                      (Rendering|SetActorHiddenInGameWithTag (Variables|Default|GetPlayerPawn) true "(TagName=\"Pawn.HidePlayerPawn.HidingProp\")" (Variables|Default|GetHidingProp) true))))))))))))
                                (elif (Variables|/Game/Athena/Items/EnvironmentalItems/HidingProps/B_HidingProp.B_HidingProp_C|GetMoveToActorOnEnter (Variables|Default|GetHidingProp))
                                  (Ability|Tasks|ApplyRootMotionMoveToActorForce "None" (Variables|Default|GetHidingProp) "0, 0, 0" "AlignFromTargetToSource" _returnvalue_8 0 0 false "MOVE_Custom" false 0 0 "SetVelocity"
                                    (:then
                                      _)
                                    (:OnFinished
                                      _))
                                  (else
                                    (Ability|Tasks|ApplyRootMotionMoveToForce "None" (Transformation|GetWorldLocation (Variables|/Game/Athena/Items/EnvironmentalItems/HidingProps/B_HidingProp.B_HidingProp_C|GetHideLocation_ForwardVector (Variables|Default|GetHidingProp))) _returnvalue_8 false "MOVE_Custom" false 0 "SetVelocity"
                                      (:then
                                        ;;[1 0.9558 0.989] Set rotation during entrance move to.
                                        (if (Variables|/Game/Athena/Items/EnvironmentalItems/HidingProps/B_HidingProp.B_HidingProp_C|GetSetEntranceRotation (Variables|Default|GetHidingProp))
                                          (Transformation|SetActorRotation (Variables|Default|GetPlayerPawn) (Math|Rotator|CombineRotators (Transformation|GetWorldRotation (Variables|/Game/Athena/Items/EnvironmentalItems/HidingProps/B_HidingProp.B_HidingProp_C|GetHideLocation_ForwardVector (Variables|Default|GetHidingProp))) "0,180.000000, 0") true)))
                                      (:OnTimedOut
                                        _)
                                      (:OnTimedOutAndDestinationReached
                                        _))))))
                            (:Is%20Not%20Valid
                              ;;[1 0.1927 0.2283] EndAbility - the avatar actor is invalid
                              (Fortnite|FortShippingLog self "PlayerPawn (AvatarActor) is not valid!" true)
                              (Ability|EndAbility))))
                        (:then_1
                          ;;[1 0.6822 0] Camera Mode applied immediately
                          (Camera|PushCameraMode (Variables|/Game/Athena/Items/EnvironmentalItems/HidingProps/B_HidingProp.B_HidingProp_C|GetCamera%20Mode (Variables|Default|GetHidingProp))))
                        (:then_2
                          ;;[1 1 1] Ignore collision with the pawn immediately
                          (Collision|AddIgnoredActor (Variables|Default|GetPlayerPawn) (Variables|Default|GetHidingProp) 0.0)
                          (if (Variables|/Game/Athena/Items/EnvironmentalItems/HidingProps/B_HidingProp.B_HidingProp_C|GetIgnorePhysicsBodyCollisionOnEnter (Variables|Default|GetHidingProp))
                            (Collision|SetCollisionResponsetoChannel (Variables|/Script/Engine.Character|GetCapsuleComponent (Variables|Default|GetPlayerPawn)) "ECC_PhysicsBody" "ECR_Ignore")))))
                    (:then_2
                      _)))
                (:CastFailed
                  (Ability|EndAbility))))
            (:Event.Athena.HidingProp.LandedOn
              (bind _asb_hiding_prop_3 (Utilities|Casting|CastToB_HidingProp _instigator)
                (:then
                  (bind _output_get_4 (Variables|Default|SetHidingProp _asb_hiding_prop_3))
                  ;[1 1 1] Remove InputComponent On Successful Teleport from
                  ; Teleport Startpoint On Client
                  (if (not (== _output_get_4 _previnputenabledhidingprop))
                    (Utilities|IsValid _previnputenabledhidingprop
                      (:Is%20Valid
                        (HidingProp|SetInputEnabled _previnputenabledhidingprop (Variables|Default|GetPlayerPawn) false))
                      (:Is%20Not%20Valid
                        _)))
                  (Utilities|FlowControl|Sequence
                    (:then_0
                      (Ability|ApplyGameplayEffectToOwner "/Game/Athena/Items/EnvironmentalItems/HidingProps/GE_Athena_HidingProp_HidingInProp.GE_Athena_HidingProp_HidingInProp_C")
                      ;[1 1 1] Update The Hiding Prop of our Entrance
                      (Class|BHidingProp|AddHiddenPlayer (Variables|Default|GetHidingProp) (Variables|Default|GetPlayerPawn))
                      (Class|BHidingProp|HidingPlayerCountChanged (Variables|Default|GetHidingProp))
                      ;[1 1 1] If Landed on, instantly put player in prop
                      (if (Variables|Default|GetAllowCosmetics)
                        (CallFunction|WobbleHidingProp (ScalableFloat|GetValueatLevel (Variables|/Game/Athena/Items/EnvironmentalItems/HidingProps/B_HidingProp.B_HidingProp_C|GetEnterWobbleRadius (Variables|Default|GetHidingProp))))
                        ;;[1 0.822 0] Camera Mode Applied
                        (Camera|PushCameraMode (Variables|/Game/Athena/Items/EnvironmentalItems/HidingProps/B_HidingProp.B_HidingProp_C|GetCamera%20Mode (Variables|Default|GetHidingProp)))
                        (else
                          _))
                      (Variables|Default|SetLandedOnProp true))
                    (:then_1
                      _)))
                (:CastFailed
                  (Ability|EndAbility))))
            (:Event.Athena.HidingProp.StopHiding
              (if (not (Networking|IsServer))
                (HidingProp|SetInputEnabled (Variables|Default|GetHidingProp) (Variables|Default|GetPlayerPawn) false))
              (CallFunction|SetCosmeticsAllowed true)
              ;[0 0.1243 1] FX Dept: Make CCPM meshes and particles visible. -Andrew Melnychuk and Eric Ketchum
              (bind _asplayer_pawn_athena (Utilities|Casting|CastToPlayerPawn_Athena (Variables|Default|GetPlayerPawn)))
              ; Cosmetics route through C++ to also be hidden so this is a double toggle without respect to existing flags/refcounts. Remove IsHiddingInProp called with rollback cleanup
              (bind _returnvalue_11 (Development|GetConsoleVariableBoolValue "Fort.Rollback.ReassertCharacterPartHidesOnActorShow"))
              (if (not _returnvalue_11)
                (Class|PlayerPawnAthena|IsHiddinginProp _asplayer_pawn_athena false))
              ;[1 1 1] Exit Prop
              (if (not (Variables|/Game/Athena/Items/EnvironmentalItems/HidingProps/B_HidingProp.B_HidingProp_C|GetSkipRestoreCameraViewTargetOnStopHiding (Variables|Default|GetHidingProp)))
                ;;[1 1 1] Retach Camera to Player
                (Game|Player|SetViewTargetwithBlend (Fortnite|GetFortPlayerControllerfromActor (Variables|Default|GetPlayerPawn)) (Variables|Default|GetPlayerPawn) 0.5 "VTBlend_EaseIn" 4.0 false)
                (else
                  _))
              (Components|Movement|StopMovementImmediately _charactermovement_2)
              (Pawn|Components|CharacterMovement|SetMovementMode _charactermovement_2 "MOVE_Walking" 0)
              ;;[1 0.8791 0] clear camera mode
              (Camera|PopCameraMode)
              (Ability|RemoveGameplayEffectFromOwnerWithGrantedTags (Variables|Tags|GetHidingInPropTag))
              (Ability|RemoveGameplayEffectFromOwnerWithGrantedTags (Variables|Tags|GetGettingInPropTag))
              (CallFunction|SetAvatarHiddeninGame false)
              (Utilities|IsValid _hidingprop_4
                (:Is%20Valid
                  (Class|BHidingProp|RemoveHiddenPlayer _hidingprop_4 (Variables|Default|GetPlayerPawn))
                  (Class|BHidingProp|HidingPlayerCountChanged (Variables|Default|GetHidingProp))
                  ;;[1 1 1] Detach Player from Hiding Prop
                  (bind _returnvalue_12 (FortPhysicsSimulationLibrary|CanActorSimulatePhysics (Variables|Default|GetHidingProp)))
                  (if _returnvalue_12
                    (Utilities|Time|ClearandInvalidateTimerbyHandle (Variables|Default|GetLockOnTimer))
                    (Collision|SetActorEnableCollision (Variables|Default|GetPlayerPawn) true)
                    (Class|BHidingProp|IgnorePhysicsCollisionDamage (Variables|Default|GetHidingProp) (Variables|Default|GetPlayerPawn) 3.0)
                    (else
                      _))
                  ;;[1 1 1] Launch other players off top of hiding prop.
                  (Class|BHidingProp|LaunchPlayersOffTop (Variables|Default|GetHidingProp) (Variables|Default|GetPlayerPawn))
                  ;;[1 1 1] Ignore hidng prop collision multicast
                  (Class|BHidingProp|IgnorePawnCollision (Variables|Default|GetHidingProp) (Variables|Default|GetPlayerPawn) (Math|Float|SelectFloat 0.2 0.8 (Variables|/Game/Athena/Items/EnvironmentalItems/HidingProps/B_HidingProp.B_HidingProp_C|GetJumpOut (Variables|Default|GetHidingProp))))
                  (if (Variables|/Game/Athena/Items/EnvironmentalItems/HidingProps/B_HidingProp.B_HidingProp_C|GetJumpOut (Variables|Default|GetHidingProp))
                    (AthenaMovement|SetPendingSkydiveLaunch _playerpawn_5 false)
                    (Character|LaunchCharacter _playerpawn_5 (* (Math|Vector|SelectVector (+ (* (Transformation|GetForwardVector (Variables|/Game/Athena/Items/EnvironmentalItems/HidingProps/B_HidingProp.B_HidingProp_C|GetHideLocation_ForwardVector (Variables|Default|GetHidingProp))) (Variables|/Game/Athena/Items/EnvironmentalItems/HidingProps/B_HidingProp.B_HidingProp_C|GetExitLaunchVelocity (Variables|Default|GetHidingProp))) (Variables|/Game/Athena/Items/EnvironmentalItems/HidingProps/B_HidingProp.B_HidingProp_C|GetAdditionalLaunchVelocity (Variables|Default|GetHidingProp))) (+ (Variables|/Game/Athena/Items/EnvironmentalItems/HidingProps/B_HidingProp.B_HidingProp_C|GetAdditionalLaunchVelocity (Variables|Default|GetHidingProp)) (* (Math|Vector|SelectVector (Math|Vector|GetForwardVector (Transformation|GetActorRotation (Variables|Default|GetPlayerPawn))) (Math|Vector|Normalize (Math|Vector|MakeVector (.x (Math|Vector|GetForwardVector (Pawn|GetControlRotation (Variables|Default|GetPlayerPawn) :!class /Script/Engine.Pawn))) (.y (Math|Vector|GetForwardVector (Pawn|GetControlRotation (Variables|Default|GetPlayerPawn) :!class /Script/Engine.Pawn))))) (<= (Math|Vector|VectorLength (Math|Vector|MakeVector (.x (Math|Vector|GetForwardVector (Pawn|GetControlRotation (Variables|Default|GetPlayerPawn) :!class /Script/Engine.Pawn))) (.y (Math|Vector|GetForwardVector (Pawn|GetControlRotation (Variables|Default|GetPlayerPawn) :!class /Script/Engine.Pawn))))) 0.5)) (Variables|/Game/Athena/Items/EnvironmentalItems/HidingProps/B_HidingProp.B_HidingProp_C|GetExitLaunchVelocity (Variables|Default|GetHidingProp)))) (Variables|/Game/Athena/Items/EnvironmentalItems/HidingProps/B_HidingProp.B_HidingProp_C|GetFixedEntranceDirection (Variables|Default|GetHidingProp))) _eventmagnitude) true true)
                    (Collision|SetActorEnableCollision _playerpawn_2 true)
                    (Utilities|IsValid (Variables|Default|GetHidingProp)
                      (:Is%20Valid
                        ;;[1 1 1] If prop was destroyed, play animation
                        (if _returnvalue_7
                          (Animation|PlayAnimMontage _playerpawn_2 "/Game/Animation/Game/MainPlayer/Reacts/WorldInteraction/Hideable/Surprise_Montage.Surprise_Montage" 1.0 "None")
                          (else
                            (Animation|PlayAnimMontage _playerpawn_2 (Variables|/Game/Athena/Items/EnvironmentalItems/HidingProps/B_HidingProp.B_HidingProp_C|GetExitAnimMontage (Variables|Default|GetHidingProp)) 1.0 "None")
                            (CallFunction|WobbleHidingProp (ScalableFloat|GetValueatLevel (Variables|/Game/Athena/Items/EnvironmentalItems/HidingProps/B_HidingProp.B_HidingProp_C|GetEnterWobbleRadius (Variables|Default|GetHidingProp))))))
                        (if (Networking|IsServer)
                          (Utilities|IsValid (Variables|Default|GetHidingProp)
                            (:Is%20Valid
                              (Ability|ExecuteGameplayCueWithParamsOnOwner (Variables|/Game/Athena/Items/EnvironmentalItems/HidingProps/B_HidingProp.B_HidingProp_C|GetExitGameplayCue (Variables|Default|GetHidingProp)) (Ability|GameplayCue|MakeGameplayCueParameters :NormalizedMagnitude 0.0 :RawMagnitude 0.0 :Location (Math|Vector|SelectVector (Transformation|GetActorLocation (Variables|Default|GetPlayerPawn)) (Transformation|GetWorldLocation (Variables|/Game/Athena/Items/EnvironmentalItems/HidingProps/B_HidingProp.B_HidingProp_C|GetHideLocation_ForwardVector (Variables|Default|GetHidingProp))) _returnvalue_7) :Normal "0,0,0" :Instigator (Variables|Default|GetHidingProp) :GameplayEffectLevel 1 :AbilityLevel 1 :bReplicateLocationWhenUsingMinimalRepProxy false))
                              (Ability|ExecuteGameplayCueWithParamsOnOwner (Variables|/Game/Athena/Items/EnvironmentalItems/HidingProps/B_HidingProp.B_HidingProp_C|GetExitGameplayCue_Player (Variables|Default|GetHidingProp)) (Ability|GameplayCue|MakeGameplayCueParameters :NormalizedMagnitude 0.0 :RawMagnitude 0.0 :Location (Transformation|GetActorLocation (Variables|Default|GetPlayerPawn)) :Normal (Transformation|GetActorForwardVector (Variables|Default|GetPlayerPawn)) :Instigator (Variables|Default|GetHidingProp) :GameplayEffectLevel 1 :AbilityLevel 1 :bReplicateLocationWhenUsingMinimalRepProxy false))
                              (CallFunction|SetAvatarHiddeninGame false)
                              (Collision|SetActorEnableCollision (Variables|Default|GetPlayerPawn) true)
                              (Networking|ForceNetUpdate (Variables|Default|GetPlayerPawn))
                              (Utilities|IsValid _hidingprop_6
                                (:Is%20Valid
                                  (if (Variables|/Game/Athena/Items/EnvironmentalItems/HidingProps/B_HidingProp.B_HidingProp_C|GetJumpOut _hidingprop_6)
                                    (Utilities|Time|SetTimerbyEvent (delegate Custom|End) (ScalableFloat|GetValueatLevel (Variables|Default|GetExitBlockShootDelay)))
                                    (Ability|ApplyGameplayEffectToOwner "/Game/Athena/Items/EnvironmentalItems/HidingProps/GE_Athena_HidingProp_ExitingProp.GE_Athena_HidingProp_ExitingProp_C")
                                    (else
                                      (Ability|ApplyGameplayEffectToOwner (Variables|/Game/Athena/Items/EnvironmentalItems/HidingProps/B_HidingProp.B_HidingProp_C|GetGE_OnExitingPropNoJump_BlockActions (Variables|Default|GetHidingProp))) ; Props that don't jump out will handle ending in their own time\nbut still want to apply their own Exiting Prop GE.
                                    )))
                                (:Is%20Not%20Valid
                                  _)))
                            (:Is%20Not%20Valid
                              _))
                          (Ability|ApplyGameplayEffectToOwner "/Game/Athena/Items/EnvironmentalItems/HidingProps/GE_Athena_HidingProp_ExitingProp.GE_Athena_HidingProp_ExitingProp_C")
                          (CallFunction|End)
                          (else
                            _)))
                      (:Is%20Not%20Valid
                        _))
                    (elif _returnvalue_7
                      ;;[1 1 1] Prop destroyed = surprise montage
                      (Animation|PlayAnimMontage (Variables|Default|GetPlayerPawn) "/Game/Animation/Game/MainPlayer/Reacts/WorldInteraction/Hideable/Surprise_Montage.Surprise_Montage" 1.0 "None")
                      ;;[1 1 1] Non Jump Exit - make this use exit anim again and make phonebooth an exception in 12.10
                      (CallFunction|GetAvatarasValidFortPlayerPawn
                        (:ValidFortPlayerPawn
                          (if (Variables|/Game/Athena/Items/EnvironmentalItems/HidingProps/B_HidingProp.B_HidingProp_C|GetSkipRootMotionMovementOnStopHiding (Variables|Default|GetHidingProp))
                            (Utilities|FlowControl|Sequence
                              (:then_0
                                _)
                              (:then_1
                                (Utilities|FlowControl|Delay 0.5)
                                (Ability|RemoveGameplayEffectFromOwnerWithGrantedTags (Variables|Tags|GetHidingInPropTag))
                                (Ability|RemoveGameplayEffectFromOwnerWithGrantedTags (Variables|Tags|GetGettingInPropTag))
                                (Ability|EndAbility)))
                            (else
                              (Ability|Tasks|ApplyRootMotionMoveToForce "None" (+ (* (Transformation|GetForwardVector (Variables|/Game/Athena/Items/EnvironmentalItems/HidingProps/B_HidingProp.B_HidingProp_C|GetHideLocation_ForwardVector (Variables|Default|GetHidingProp))) (Math|Float|SelectFloat 0.01 (Variables|/Game/Athena/Items/EnvironmentalItems/HidingProps/B_HidingProp.B_HidingProp_C|GetNonJumpExitDistance (Variables|Default|GetHidingProp)) _returnvalue_7)) (Transformation|GetActorLocation (Variables|Default|GetHidingProp))) 0.7 false "MOVE_Custom" false 0 "SetVelocity"
                                (:then
                                  _)
                                (:OnTimedOut
                                  _)
                                (:OnTimedOutAndDestinationReached
                                  _)))))
                        (:AvatarCastFailed
                          (Ability|EndAbility))) ; Ensure we have a valid player pawn at this point \n(Applying Root Motion can crash the server at this point if not)
                      (else
                        ;;[1 1 1] Non Jump exit montage; used by tent but not phonebooth
                        (Utilities|IsValid _exitanimmontage
                          (:Is%20Valid
                            (Animation|PlayAnimMontage (Variables|Default|GetPlayerPawn) _exitanimmontage 1.0 "None"))
                          (:Is%20Not%20Valid
                            _))))))
                (:Is%20Not%20Valid
                  _))))
          (bind _returnvalue_13 (Fortnite|HeldObjects|GetHeldObjectComponentAttachedtoPlayer (Variables|Default|GetPlayerPawn)))
          (HeldObjects|Behavior|DropHeldObject _returnvalue_13)
          ;[1 0.7554 0.3604] If player goes dbno, starts carrying a player, or puts on a snowman item, stophiding and cancel the ability.
          (Utilities|IsServer
            (:true
              (Utilities|FlowControl|Sequence
                (:then_0
                  (Ability|Tasks|WaitGameplayTagAdd "(TagName=\"Gameplay.Status.DBNO\")"
                    (:Added
                      _)))
                (:then_1
                  (Ability|Tasks|WaitGameplayTagAdd "(TagName=\"Gameplay.Action.Player.CarryingPlayer\")"
                    (:Added
                      _)))
                (:then_2
                  (Ability|Tasks|WaitGameplayTagAdd "(TagName=\"Athena.Item.Snowman\")"
                    (:Added
                      _)))
                (:then_3
                  (Ability|Tasks|WaitGameplayTagAdd "(TagName=\"Gameplay.Action.Player.GliderRedeploy\")"
                    (:Added
                      _)))
                (:then_4
                  (Ability|Tasks|WaitGameplayTagAdd "(TagName=\"Gameplay.Action.Player.Skydive\")"
                    (:Added
                      _)))
                (:then_5
                  (Ability|Tasks|WaitGameplayTagAdd "(TagName=\"MovementMode.Parachuting\")"
                    (:Added
                      _)))
                (:then_6
                  (Ability|Tasks|WaitGameplayEvent "(TagName=\"Abilities.Generic.Rift.Trigger\")"
                    (:EventReceived
                      (Ability|SendGameplayEvent "(TagName=\"Event.Athena.HidingProp.StopHiding\")" (Utilities|Struct|MakeGameplayEventData :ContextHandle "()" :InstigatorTags "(GameplayTags=)" :TargetTags "(GameplayTags=)" :EventMagnitude 0.8 :TargetData "()")))))
                (:then_7
                  (Ability|Tasks|WaitGameplayTagAdd "(TagName=\"Gameplay.Status.Rift.Teleported\")"
                    (:Added
                      _)))
                (:then_8
                  (Ability|Tasks|WaitGameplayTagAdd (Variables|Default|GetCurieBurningTag) (Variables|Default|GetHidingProp)
                    (:Added
                      _)))))))
        (:Is%20Not%20Valid
          _)))
    (:CastFailed
      _))
  (Ability|EndAbility))

(event Custom|WobbleHidingProp
  (params (NormalizedMagnitude :type Float%20%28double-precision%29))
  ;[1 0.8813 0] General Purpose Wobble; requires complex params for normal & location
  (Ability|ExecuteGameplayCueWithParamsOnOwner (Variables|FX|GetGC_Wobble) (Ability|GameplayCue|MakeGameplayCueParameters :NormalizedMagnitude NormalizedMagnitude :RawMagnitude 0.0 :Location (Variables|/Game/Athena/Items/EnvironmentalItems/HidingProps/B_HidingProp.B_HidingProp_C|GetWobbleLocationOffset (Variables|Default|GetHidingProp)) :Normal (select (Variables|/Game/Athena/Items/EnvironmentalItems/HidingProps/B_HidingProp.B_HidingProp_C|GetRandomWobbleNormal (Variables|Default|GetHidingProp)) (Math|Rotator|GetRotationXVector (Math|Random|RandomRotator true)) (* (Transformation|GetForwardVector (Variables|/Game/Athena/Items/EnvironmentalItems/HidingProps/B_HidingProp.B_HidingProp_C|GetHideLocation_ForwardVector (Variables|Default|GetHidingProp))) 0.25)) :Instigator (Variables|Default|GetHidingProp) :GameplayEffectLevel 1 :AbilityLevel 1 :bReplicateLocationWhenUsingMinimalRepProxy false)))

(event Custom|End
  ;[1 1 1] Exit Prop
  (Ability|RemoveGameplayEffectFromOwnerWithGrantedTags (Variables|Tags|GetHidingInPropTag))
  (Ability|RemoveGameplayEffectFromOwnerWithGrantedTags (Variables|Tags|GetGettingInPropTag))
  (Ability|EndAbility))

(fn DebugAbilityLogging
  (params (BuildLogEnabled? :type Boolean :default true) (PIE%20PrintEnabled? :type Boolean) (Message :type String) (GameplayEventData :type Gameplay%20Event%20Data%20Structure%20%28by%20ref%29))
  (bind _instigator (Utilities|Struct|BreakGameplayEventData GameplayEventData))
  (bind _returnvalue (Utilities|GetDisplayName _instigator))
  (bind _returnvalue_2 (Utilities|String|Append Message _returnvalue))
  (Utilities|FlowControl|Sequence
    (:then_0
      (if BuildLogEnabled?
        (Fortnite|FortShippingLog self _returnvalue_2)))
    (:then_1
      (if PIE%20PrintEnabled?
        (Development|PrintString _returnvalue_2)))))

(fn SetAvatarHiddenInGame
  (params (New%20Hidden :type Boolean))
  (bind _output_get (LocalVariables|SetAvatarActor (Ability|GetAvatarActorfromActorInfo)))
  (Utilities|IsValid _output_get
    (:Is%20Valid
      (bind _asfort_player_pawn (Utilities|Casting|CastToFortPlayerPawn (LocalVariables|GetAvatarActor))
        (:then
          (Rendering|SetActorHiddenInGameWithTag _asfort_player_pawn (Variables|GetNew%20Hidden) "(TagName=\"Pawn.HidePlayerPawn.HidingProp\")" (Variables|Default|GetHidingProp) true))
        (:CastFailed
          (Rendering|SetActorHiddenInGame (LocalVariables|GetAvatarActor) (Variables|GetNew%20Hidden)))))))

(fn SetCosmeticsAllowed
  (params (Allowed :type Boolean))
  (Ability|HasAuthority
    (:True
      (if (Math|Boolean|NotEqual(Boolean) (Variables|Default|GetAllowCosmetics) Allowed)
        (Variables|Default|SetAllowCosmetics Allowed)
        (bind _playerpawn (Variables|Default|GetPlayerPawn)
          (:then
            (bind _hidingprop (Variables|Default|GetHidingProp)
              (:then
                (if Allowed
                  (Utilities|Array|RemoveItem (Variables|/Game/Athena/Items/EnvironmentalItems/HidingProps/B_HidingProp.B_HidingProp_C|GetNonCosmeticPawns _hidingprop) _playerpawn)
                  (else
                    (Utilities|Array|AddUnique (Variables|/Game/Athena/Items/EnvironmentalItems/HidingProps/B_HidingProp.B_HidingProp_C|GetNonCosmeticPawns _hidingprop) _playerpawn))))
              (:else)))
          (:else))))))
