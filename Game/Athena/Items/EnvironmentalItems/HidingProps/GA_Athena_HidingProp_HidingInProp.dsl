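; Gameplay ability that runs while a player hides inside a hiding prop.
; With the HidingInProp tag present on activation, it optionally disables collision, plays a looping hiding montage and unblocks the change-equipment tag as a grappler failsafe.
; Without the tag, a timer later ends the ability if the tag is still missing, working around briefly applied effects.
; Ending stops the montage and restores collision when that setting is on.
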(event Custom|GrapplerFailSafe
  (bind _returnvalue (Ability|GetAvatarActorfromActorInfo))
  (bind _returnvalue_2 (Ability|GetAbilitySystemComponent _returnvalue))
  ;[0.6158 0.3512 1] Grappler Failsafe for FORT-298499.  There's some replication breakdown keeping this from being solved via ability tags, so doing this instead
  (Utilities|IsValid _returnvalue_2
    (:Is%20Valid
      (bind _asfort_ability_system_component (Utilities|Casting|CastToFortAbilitySystemComponent _returnvalue_2))
      (Tags|UnBlockAbilitiesWithTags _asfort_ability_system_component (Variables|Default|GetChangeEquipmentTag) false))))

(event Ability|EventActivateAbility
  (bind _returnvalue (Ability|GetAvatarActorfromActorInfo))
  ;[0.684 0.2219 1] Workaround for FORT-629575 (Granting GE gets added/removed quickly, this GA lingers)
  ; If we have the HidingInProp tag, good to go. If not and it does not show up in X seconds, then end the ability.
  ; Also added the HidingInProp tag to CancelSelfOnTagsChanged, so if the tag is removed mid ability, the ability will also end. (This was not enough by itself as the tag can be added & removed before the ability activates.)
  ;
  ; Note: This could potentially also be accomplished by adding the HidingInProp Tag to the Activation Required tags. However, I wasn't comfortable doing that due to potential timing differences on when the GE adds/replicates the tag and the GA activates.
  (bind _asgameplay_tag_asset_interface (Utilities|Casting|CastToGameplayTagAssetInterface (Ability|GetAvatarActorfromActorInfo)))
  (if (GameplayTags|HasMatchingGameplayTag _asgameplay_tag_asset_interface "(TagName=\"Granted.Athena.EnvItem.HidingProp.HidingInProp\")")
    (if (ScalableFloat|GetValueasBool (Variables|Default|GetRow_bDisablePlayerCollision))
      (Collision|SetActorEnableCollision _returnvalue false)
      (else
        _))
    (bind _asfort_player_pawn (Utilities|Casting|CastToFortPlayerPawn _returnvalue))
    (Animation|PlayAnimMontage _asfort_player_pawn "/Game/Animation/Game/MainPlayer/Reacts/WorldInteraction/Hideable/Hiding_HidingLoopMontage.Hiding_HidingLoopMontage" 1.0 "None")
    (CallFunction|GrapplerFailSafe)
    (else
      (Utilities|Time|SetTimerbyEvent (delegate Custom|EndAbilityIfMissingTag) (Variables|Default|GetWaitForRequiredTagTime)))))

(event Ability|EventOnEndAbility
  (params (bWasCancelled :type Boolean))
  (bind _returnvalue (Ability|GetAvatarActorfromActorInfo))
  (if (ScalableFloat|GetValueasBool (Variables|Default|GetRow_bDisablePlayerCollision))
    (Collision|SetActorEnableCollision _returnvalue true)
    (else
      _))
  (bind _asfort_player_pawn (Utilities|Casting|CastToFortPlayerPawn _returnvalue))
  (Animation|StopAnimMontage _asfort_player_pawn "/Game/Animation/Game/MainPlayer/Reacts/WorldInteraction/Hideable/Hiding_HidingLoopMontage.Hiding_HidingLoopMontage"))

(event Custom|EndAbilityIfMissingTag
  ;[0.684 0.2219 1] Workaround for FORT-629575 (Granting GE gets added/removed quickly, this GA lingers)
  ; If we have the HidingInProp tag, good to go. If not and it does not show up in X seconds, then end the ability.
  ; Also added the HidingInProp tag to CancelSelfOnTagsChanged, so if the tag is removed mid ability, the ability will also end. (This was not enough by itself as the tag can be added & removed before the ability activates.)
  ;
  ; Note: This could potentially also be accomplished by adding the HidingInProp Tag to the Activation Required tags. However, I wasn't comfortable doing that due to potential timing differences on when the GE adds/replicates the tag and the GA activates.
  (bind _asgameplay_tag_asset_interface (Utilities|Casting|CastToGameplayTagAssetInterface (Ability|GetAvatarActorfromActorInfo)))
  (if (not (GameplayTags|HasMatchingGameplayTag _asgameplay_tag_asset_interface "(TagName=\"Granted.Athena.EnvItem.HidingProp.HidingInProp\")"))
    (Ability|EndAbility) ; Still do not have tag, end ability
  ))
