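; Gameplay ability that gets a player out of an Athena hiding prop.
; CanActivateAbility requires the HidingInProp tag and rejects pawns that are DBNO, still getting in, or exit-blocked.
; On the server, a StopHiding trigger calls StopHiding on the prop and pawn carried in the event payload.
; Otherwise it finds the prop through the tagged Hide ability and, unless BypassJumpOut is set, sends StopHiding and releases the pawn.
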(event Ability|EventActivateAbilityFromEvent
  (params (EventData :type Gameplay%20Event%20Data%20Structure%20%28by%20ref%29))
  (bind (_eventtag _instigator _target _optionalobject _optionalobject2) (Utilities|Struct|BreakGameplayEventData EventData))
  (bind _returnvalue (Ability|GetAvatarActorfromActorInfo))
  (bind _returnvalue_2 (Ability|GetFortAbilitySystemComponent _returnvalue))
  (Utilities|IsServer
    (:true
      (bind _returnvalue_3 (GameplayTags|HasMatchingGameplayTag(Message) (Ability|GetAvatarActorfromActorInfo) "(TagName=\"Granted.Athena.EnvItem.HidingProp.GettingInProp\")"))
      (if _returnvalue_3
        _
        (elif (== _eventtag "(TagName=\"Event.Athena.HidingProp.StopHiding\")")
          (bind _asb_hiding_prop (Utilities|Casting|CastToB_HidingProp _optionalobject))
          (bind _asfort_pawn (Utilities|Casting|CastToFortPawn _optionalobject2))
          (Class|BHidingProp|StopHiding _asb_hiding_prop _asfort_pawn)
          (else
            (bind _returnvalue_4 (GameplayTags|HasMatchingGameplayTag(Message) (Ability|GetAvatarActorfromActorInfo) "(TagName=\"Granted.Athena.EnvItem.HidingProp.HidingInProp\")"))
            (if _returnvalue_4
              (bind _returnvalue_5 (GameplayTags|HasMatchingGameplayTag(Message) (Ability|GetAvatarActorfromActorInfo) "(TagName=\"Gameplay.Status.DBNO\")"))
              (if (not _returnvalue_5)
                (bind _outabilityhandles (GameplayAbilities|FindAllAbilitieswithTags _returnvalue_2 (Variables|Default|GetGameplayCueAthenaHidingProp) true))
                (if (Utilities|Array|IsNotEmpty _outabilityhandles)
                  (bind _asga_athena_hiding_prop_hide (Utilities|Casting|CastToGA_Athena_HidingProp_Hide (.ReturnValue (Ability|GameplayAbility|GetGameplayAbilityfromSpecHandle _returnvalue_2 (Utilities|Array|Get(acopy) _outabilityhandles))))
                    (:then
                      ;[0.1 0.1042 1] Check if this hiding prop skips jump out logic
                      (if (not (Variables|/Game/Athena/Items/EnvironmentalItems/HidingProps/B_HidingProp.B_HidingProp_C|GetBypassJumpOut (Variables|/Game/Athena/Items/EnvironmentalItems/HidingProps/GA_Athena_HidingProp_Hide.GA_Athena_HidingProp_Hide_C|GetHidingProp _asga_athena_hiding_prop_hide)))
                        (Ability|SendGameplayEvent "(TagName=\"Event.Athena.HidingProp.StopHiding\")" (Utilities|Struct|MakeGameplayEventData :Instigator (Variables|/Game/Athena/Items/EnvironmentalItems/HidingProps/GA_Athena_HidingProp_Hide.GA_Athena_HidingProp_Hide_C|GetHidingProp _asga_athena_hiding_prop_hide) :ContextHandle "()" :InstigatorTags "(GameplayTags=)" :TargetTags "(GameplayTags=)" :EventMagnitude 1.0 :TargetData "()"))
                        (bind _asfort_pawn_2 (Utilities|Casting|CastToFortPawn (Ability|GetAvatarActorfromActorInfo)))
                        (Class|BHidingProp|StopHiding (Variables|/Game/Athena/Items/EnvironmentalItems/HidingProps/GA_Athena_HidingProp_Hide.GA_Athena_HidingProp_Hide_C|GetHidingProp _asga_athena_hiding_prop_hide) _asfort_pawn_2)
                        (else
                          (Ability|EndAbility))))
                    (:CastFailed
                      _))
                  (else
                    _))
                (Ability|SendGameplayEvent "(TagName=\"Event.Athena.HidingProp.StopHiding\")" (Utilities|Struct|MakeGameplayEventData "(TagName=\"\")")))
              (else
                _))
            (Ability|EndAbility)))))))

(event Ability|EventOnEndAbility
  (params (bWasCancelled :type Boolean)))

(fn CanActivateAbility
  (params (ActorInfo_OwnerActor :type Actor%20Object%20Reference) (ActorInfo_AvatarActor :type Actor%20Object%20Reference) (ActorInfo_PlayerController :type Player%20Controller%20Object%20Reference) (ActorInfo_AbilitySystemComponent :type Ability%20System%20Component%20Object%20Reference) (ActorInfo_SkeletalMeshComponent :type Skeletal%20Mesh%20Component%20Object%20Reference) (ActorInfo_AnimInstance :type Anim%20Instance%20Object%20Reference) (ActorInfo_MovementComponent :type Movement%20Component%20Object%20Reference) (ActorInfo_AffectedAnimInstanceTag :type Name) (Handle :type Gameplay%20Ability%20Spec%20Handle%20Structure))
  (bind _output_get (LocalVariables|SetLocalAvatarActor ActorInfo_AvatarActor))
  (Utilities|IsValid _output_get
    (:Is%20Valid
      (bind _returnvalue (GameplayTags|HasMatchingGameplayTag(Message) (LocalVariables|GetLocalAvatarActor) "(TagName=\"Granted.Athena.EnvItem.HidingProp.GettingInProp\")"))
      (bind _returnvalue_2 (GameplayTags|HasMatchingGameplayTag(Message) (LocalVariables|GetLocalAvatarActor) "(TagName=\"Granted.Athena.EnvItem.HidingProp.HidingInProp\")"))
      (bind _returnvalue_3 (GameplayTags|HasMatchingGameplayTag(Message) (LocalVariables|GetLocalAvatarActor) "(TagName=\"Gameplay.Status.DBNO\")"))
      (bind _returnvalue_4 (GameplayTags|HasMatchingGameplayTag(Message) (LocalVariables|GetLocalAvatarActor) "(TagName=\"Granted.Athena.EnvItem.HidingProp.BlockExit\")"))
      (return 0 (Math|Boolean|ANDBoolean (not _returnvalue_3) _returnvalue_2 (not _returnvalue) (not _returnvalue_4))))
    (:Is%20Not%20Valid
      (return 0 false))))
