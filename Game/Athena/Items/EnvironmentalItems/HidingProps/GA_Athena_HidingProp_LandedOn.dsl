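; Gameplay ability for Athena hiding props that relays a landed-on event when hiding by landing is allowed.
; On activation from an event, the server checks two scalable-float toggles: hiding enabled and hiding enabled from landing.
; If both are true, it rebuilds the incoming event data and sends it as Event.Athena.HidingProp.LandedOn.
; EndAbility is called only on the server path, and the end-ability handler is empty.
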(event Ability|EventActivateAbilityFromEvent
  (params (EventData :type Gameplay%20Event%20Data%20Structure%20%28by%20ref%29))
  (bind (_eventtag _instigator _target _optionalobject _optionalobject2 _contexthandle _instigatortags _targettags _eventmagnitude _targetdata) (Utilities|Struct|BreakGameplayEventData EventData))
  (Utilities|IsServer
    (:true
      (if (and (ScalableFloat|GetValueasBool (Variables|Default|GetHidingEnabled)) (ScalableFloat|GetValueasBool (Variables|Default|GetHiding%20Enable%20From%20Landing)))
        (Ability|SendGameplayEvent "(TagName=\"Event.Athena.HidingProp.LandedOn\")" (Utilities|Struct|MakeGameplayEventData _eventtag _instigator _target _optionalobject _optionalobject2 _contexthandle _instigatortags _targettags _eventmagnitude _targetdata))
        (else
          _))
      (Ability|EndAbility))))

(event Ability|EventOnEndAbility
  (params (bWasCancelled :type Boolean)))
