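; Gameplay ability for Fortnite hidden passages, teleporting a player from one hiding prop to its linked target prop.
; Once the hiding tag appears, it applies effects and a camera mode, then teleports asynchronously, waiting for level streaming.
; The server wobbles the destination, ejects the player, updates quests and analytics, and grants a launch for whirlpool targets.
; Ending restores interaction, HUD and camera, applies cooldown, and unblocks equipment-change abilities as a grappler failsafe.
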(event Custom|Wobble
  ;[1 1 1] Wobble
  (Class|BHidingProp|ExecuteGameplayCue (Variables|Default|GetTargetTeleporter) (Variables|Default|GetWobbleCue) (Ability|GameplayCue|MakeGameplayCueParameters :NormalizedMagnitude 2e+02 :RawMagnitude 0.0 :Location (Math|Vector|MakeVector (.x (Variables|/Game/Athena/Items/EnvironmentalItems/HidingProps/B_HidingProp.B_HidingProp_C|GetWobbleLocationOffset (Variables|Default|GetTargetTeleporter))) (.y (Variables|/Game/Athena/Items/EnvironmentalItems/HidingProps/B_HidingProp.B_HidingProp_C|GetWobbleLocationOffset (Variables|Default|GetTargetTeleporter))) (Math|Random|RandomFloatinRange 6e+01 135.0)) :Normal (* (Math|Random|RandomUnitVector) 1.0) :Instigator (Variables|Default|GetTargetTeleporter) :GameplayEffectLevel 1 :AbilityLevel 1 :bReplicateLocationWhenUsingMinimalRepProxy false)))

(event Ability|EventFailedToActivatePassiveAbility
  (params (ActorInfo :type Gameplay%20Ability%20Actor%20Info%20Structure))
  (|CallParentFunction))

(event Ability|EventActivateAbility
  (bind _returnvalue (Ability|GetAvatarActorfromActorInfo))
  (bind _asfort_player_pawn (Utilities|Casting|CastToFortPlayerPawn _returnvalue))
  (bind _returnvalue_2 (Pawn|GetControlRotation _asfort_player_pawn :!class /Script/Engine.Pawn))
  (bind _yaw (.yaw (Transformation|GetWorldRotation (Variables|/Game/Athena/Items/EnvironmentalItems/HidingProps/B_HidingProp.B_HidingProp_C|GetHideLocation_ForwardVector (Variables|Default|GetTargetTeleporter)))))
  (bind _returnvalue_3 (Math|Rotator|MakeRotator 0.0 0.0 _yaw))
  (bind _yaw_2 (.yaw _returnvalue_2))
  (bind _targetteleporter (Variables|Default|GetTargetTeleporter))
  (bind _hidelocation_forwardvector (Variables|/Game/Athena/Items/EnvironmentalItems/HidingProps/B_HidingProp.B_HidingProp_C|GetHideLocation_ForwardVector _targetteleporter))
  (bind _returnvalue_4 (Transformation|GetWorldRotation _hidelocation_forwardvector))
  (bind _returnvalue_5 (Math|Rotator|MakeRotator 0.0 0.0 _yaw_2))
  (bind _returnvalue_6 (Ability|GetAvatarActorfromActorInfo))
  (bind _returnvalue_7 (Ability|GetAvatarActorfromActorInfo))
  (bind _asfort_pawn (Utilities|Casting|CastToFortPawn _returnvalue_6))
  (bind _asfort_pawn_2 (Utilities|Casting|CastToFortPawn _returnvalue_7))
  (bind _targetteleporter_2 (Variables|Default|GetTargetTeleporter))
  (bind _fixedentrancedirection (Variables|/Game/Athena/Items/EnvironmentalItems/HidingProps/B_HidingProp.B_HidingProp_C|GetFixedEntranceDirection _targetteleporter_2))
  (bind _returnvalue_8 (select _fixedentrancedirection _returnvalue_3 _returnvalue_5))
  (bind _hidingprop (Variables|Default|GetHidingProp))
  (bind _targetteleporter_3 (Variables|Default|GetTargetTeleporter))
  (bind _targetteleporter_4 (Variables|Default|GetTargetTeleporter))
  (bind _hidingprop_2 (Variables|Default|GetHidingProp))
  (bind _targetteleporter_5 (Variables|Default|GetTargetTeleporter))
  ;[1 1 1] Hidden passage
  (Analytics|IncrementAnalyticMatchCount "HiddenPassagesTimesEntered")
  (bind _asb_hiding_prop (Utilities|Casting|CastToB_HidingProp (Ability|EffectContext|GetSourceObject (Ability|GetGrantedbyEffectContext))))
  (bind _asfort_player_controller (Utilities|Casting|CastToFortPlayerController (Ability|GetPlayerControllerfromActorInfo)))
  (bind _returnvalue_9 (Class|FortPlayerController|BPGetLocalPlayer _asfort_player_controller))
  (bind _returnvalue_10 (Class|FortHUDContext|GetFortHUDContextbyPlayer _returnvalue_9))
  (HUD|SetHUDElementVisibility _returnvalue_10 (Variables|Default|GetT_HudElementToHide) true "None")
  (Variables|Default|SetHidingProp _asb_hiding_prop)
  (Variables|Default|SetTargetTeleporter (Variables|/Game/Athena/Items/EnvironmentalItems/HidingProps/B_HidingProp.B_HidingProp_C|GetTargetTeleporter _asb_hiding_prop))
  (Variables|/Game/Athena/Items/EnvironmentalItems/HidingProps/B_HidingProp.B_HidingProp_C|SetTeleporting true (Variables|Default|GetTargetTeleporter))
  (Variables|/Game/Athena/Items/EnvironmentalItems/HidingProps/B_HidingProp.B_HidingProp_C|SetTeleporting true (Variables|Default|GetHidingProp))
  (Variables|/Script/FortniteGame.BuildingActor|SetbAllowInteract false (Variables|Default|GetTargetTeleporter))
  (Variables|/Script/FortniteGame.BuildingActor|SetbAllowInteract false (Variables|Default|GetHidingProp))
  (Utilities|IsServer
    (:true
      (Ability|ExecuteGameplayCueOnOwner (Variables|Default|GetEnterCue)))
    (:false
      _))
  (Utilities|FlowControl|Sequence
    (:then_0
      (Ability|Tasks|WaitGameplayTagAdd (Variables|Default|GetHidingInPropTag)
        (:Added
          (Utilities|IsValid _asfort_pawn
            (:Is%20Valid
              (Utilities|Casting|CastToFortPlayerPawn _asfort_pawn)
              (Ability|AddGameplayCueToOwner (Variables|/Game/Athena/Items/EnvironmentalItems/HidingProps/B_HidingProp.B_HidingProp_C|GetLoopingTeleportingCue (Variables|Default|GetTargetTeleporter)))
              (bind _returnvalue_11 (Environment|ApplyGlobalEnvironmentGameplayEffecttoActor _asfort_pawn "/Game/Athena/Items/EnvironmentalItems/HidingProps/GE_Athena_HidingProp_Teleporting.GE_Athena_HidingProp_Teleporting_C"))
              (Variables|Default|SetTeleportingGEOnPlayer _returnvalue_11)
              (Camera|PushCameraMode "/Game/Blueprints/Camera/Athena/Athena_PlayerCameraMode_HidingProp_Teleport.Athena_PlayerCameraMode_HidingProp_Teleport_C")
              (Utilities|FlowControl|Delay 1.0)
              ;;[1 1 1] Teleport Endpoint has become invalid. Preventing null warp to 0,0,0 by canceling.
              (Utilities|IsValid _targetteleporter_5
                (:Is%20Valid
                  ;;[1 1 1] Teleport the player before putting them in the target prop so that we get the Level Streaming Complete call back
                  (Class|FortAsyncActionTeleportandWaitforLevelStreaming|CreateTeleportandWaitforLevelStreaming (Utilities|Casting|CastToFortPlayerPawn _asfort_pawn) (Transformation|GetActorLocation _targetteleporter_5) _returnvalue_8
                    (:then
                      (if (Networking|IsServer)
                        ;;[1 1 1] Place player in target prop, Start timer for analytics.
                        (Ability|SendGameplayEventtoActor _asfort_pawn "(TagName=\"Event.Athena.HidingProp.LandedOn\")" (Utilities|Struct|MakeGameplayEventData :Instigator (Variables|Default|GetTargetTeleporter) :OptionalObject (Variables|Default|GetTargetTeleporter) :ContextHandle "()" :InstigatorTags "(GameplayTags=)" :TargetTags "(GameplayTags=)" :EventMagnitude 0.0 :TargetData "()"))
                        (Variables|Default|SetServerWorldTime (Fortnite|Sync|Time|GetServerWorldTimeSeconds))
                        (else
                          _))
                      (Variables|/Script/FortniteGame.BuildingActor|SetbAllowInteract true (Variables|Default|GetHidingProp))
                      (Utilities|Array|RemoveItem (Variables|/Game/Athena/Items/EnvironmentalItems/HidingProps/B_HidingProp.B_HidingProp_C|GetHidingPlayers _hidingprop) _asfort_pawn)
                      (Class|BHidingProp|HidingPlayerCountChanged _hidingprop))
                    (:OnComplete
                      (Camera|PopCameraMode)
                      (Transform|Actor|TeleportPlayerPawn (Utilities|Casting|CastToFortPlayerPawn (Ability|GetAvatarActorfromActorInfo)) (Transformation|GetActorLocation (Utilities|Casting|CastToFortPlayerPawn (Ability|GetAvatarActorfromActorInfo))) _returnvalue_8 true true)
                      (if (Networking|IsServer)
                        (bind _returnvalue_12 (Utilities|Time|SetTimerbyEvent (delegate Custom|Wobble) 0.3 true))
                        (Variables|Default|SetWobbleHandle _returnvalue_12)
                        ;;[1 1 1] Gameplay Cue that plays on hiding prop player will arrive at
                        (Class|BHidingProp|AddGameplayCue (Variables|Default|GetTargetTeleporter) (Variables|Default|GetTargetPropTeleportingCue) (Ability|GameplayCue|MakeGameplayCueParameters 0.0))
                        (Utilities|FlowControl|Delay 1.0)
                        (Ability|RemoveGameplayEffectFromOwnerWithHandle (Variables|Default|GetTeleportingGEOnPlayer))
                        (Ability|RemoveGameplayCueFromOwner (Variables|/Game/Athena/Items/EnvironmentalItems/HidingProps/B_HidingProp.B_HidingProp_C|GetLoopingTeleportingCue (Variables|Default|GetTargetTeleporter)))
                        (Utilities|FlowControl|Delay 0.45)
                        (Utilities|Time|ClearandInvalidateTimerbyHandle (Variables|Default|GetWobbleHandle))
                        ;;[1 1 1] Stop Gameplay Cue on Hiding prop player arrives at
                        (Class|BHidingProp|RemoveGameplayCue (Variables|Default|GetTargetTeleporter) (Variables|Default|GetTargetPropTeleportingCue) (Ability|GameplayCue|MakeGameplayCueParameters 0.0))
                        (CallFunction|QuestUpdate _asfort_pawn)
                        (Class|BHidingProp|StopHiding _targetteleporter_4 _asfort_pawn)
                        (Class|BHidingProp|HiddenTravelCompleteEvent (Variables|Default|GetHidingProp) _targetteleporter_4 _asfort_pawn)
                        (Class|BHidingProp|RemoveHiddenPlayer _hidingprop_2 _asfort_pawn)
                        (Class|BHidingProp|HidingPlayerCountChanged _hidingprop_2)
                        ;;[1 0 0.03471] Band Aid fix so that when exiting a willie pete after teleporting you are launched in the same way as when using a willie piete. Grant the williepete launch GE if target teleporter is Willie Pete
                        (if (GameplayTags|HasMatchingGameplayTag _targetteleporter_3 "(TagName=\"Building.Type.Prop.Hiding.Whirlpool\")")
                          (GameplayEffects|FortApplyGameplayEffectToTarget :self (Ability|GetFortAbilitySystemComponent _targetteleporter_3) :GameplayEffectClass "/Game/Athena/Items/EnvironmentalItems/HidingProps/Props/GE_WilliePete_PlayerLaunch_Granted.GE_WilliePete_PlayerLaunch_Granted_C" :TargetOfEffect (Ability|GetFortAbilitySystemComponent (Ability|GetAvatarActorfromActorInfo)) :OptionalSourceObject (Variables|Default|GetHidingProp) :Level 0.0)
                          (else
                            _))
                        (if (Networking|IsServer)
                          (Analytics|FireAnalyticEventEnvironmentalItemUsed (Fortnite|GetFortPlayerControllerfromActor (Ability|GetAvatarActorfromActorInfo)) (Variables|Default|GetHidingProp) (- (Fortnite|Sync|Time|GetServerWorldTimeSeconds) (Variables|Default|GetServerWorldTime)) "None" (Variables|Default|GetTargetTeleporter))
                          (else
                            _))
                        ;;[1 1 1] Time after leaving prop before you can move and interact/ Before the ability ends
                        (Utilities|FlowControl|Delay 0.5)
                        (Networking|FlushNetDormancy (Variables|Default|GetHidingProp))
                        (Ability|EndAbility)))))
                (:Is%20Not%20Valid
                  (Development|PrintString "TargetTeleporter has become invalid on ther server since the Teleport call. Canceling Teleport.")
                  (Ability|EndAbility))))))))
    (:then_1
      ;[1 1 1] Cancel everything if player gets any of these tags
      (Ability|Tasks|WaitGameplayTagAdd "(TagName=\"Gameplay.Status.DBNO\")"
        (:Added
          _)))
    (:then_2
      (Ability|Tasks|WaitGameplayTagAdd "(TagName=\"Gameplay.Action.Player.CarryingPlayer\")"
        (:Added
          _)))
    (:then_3
      (Ability|Tasks|WaitGameplayTagAdd "(TagName=\"Athena.Item.Snowman\")"
        (:Added
          _)))
    (:then_4
      (Utilities|IsValid _asfort_pawn_2
        (:Is%20Valid
          (EventDispatchers|BindEventtoOnDied _asfort_pawn_2 (delegate Custom|OnPawnDied)))))
    (:then_5
      (Ability|Tasks|WaitGameplayEvent "(TagName=\"Event.Athena.HidingProp.AbortTeleporting\")"
        (:EventReceived
          (Ability|EndAbility)))))
  (Class|BHidingProp|StopHiding (Variables|Default|GetTargetTeleporter) _asfort_pawn)
  (Class|BHidingProp|StopHiding (Variables|Default|GetHidingProp) _asfort_pawn))

(event Ability|EventOnEndAbility
  (params (bWasCancelled :type Boolean))
  (bind _returnvalue (Ability|GetAvatarActorfromActorInfo))
  (bind _asfort_pawn (Utilities|Casting|CastToFortPawn _returnvalue))
  ;[1 1 1] Hidden passage
  (Variables|/Game/Athena/Items/EnvironmentalItems/HidingProps/B_HidingProp.B_HidingProp_C|SetTeleporting false (Variables|Default|GetHidingProp))
  (Variables|/Game/Athena/Items/EnvironmentalItems/HidingProps/B_HidingProp.B_HidingProp_C|SetTeleporting false (Variables|Default|GetTargetTeleporter))
  (Variables|/Script/FortniteGame.BuildingActor|SetbAllowInteract true (Variables|Default|GetTargetTeleporter))
  (Variables|/Script/FortniteGame.BuildingActor|SetbAllowInteract true (Variables|Default|GetHidingProp))
  (Utilities|Time|ClearandInvalidateTimerbyHandle (Variables|Default|GetWobbleHandle))
  (bind _asfort_player_controller (Utilities|Casting|CastToFortPlayerController (Ability|GetPlayerControllerfromActorInfo)))
  (bind _returnvalue_2 (Class|FortPlayerController|BPGetLocalPlayer _asfort_player_controller))
  (bind _returnvalue_3 (Class|FortHUDContext|GetFortHUDContextbyPlayer _returnvalue_2))
  (HUD|SetHUDElementVisibility _returnvalue_3 (Variables|Default|GetT_HudElementToHide) false "None")
  (Ability|ApplyGameplayEffectToOwner "/Game/Athena/Items/EnvironmentalItems/HidingProps/GE_Athena_HidingProp_TeleportCoolDown.GE_Athena_HidingProp_TeleportCoolDown_C" 0)
  (Utilities|IsValid _asfort_pawn
    (:Is%20Valid
      (EventDispatchers|UnbindallEventsfromOnDied _asfort_pawn)
      (CallFunction|GrapplerFailSafe _asfort_pawn)
      (Ability|RemoveGameplayCueFromOwner (Variables|/Game/Athena/Items/EnvironmentalItems/HidingProps/B_HidingProp.B_HidingProp_C|GetLoopingTeleportingCue (Variables|Default|GetTargetTeleporter)))
      (Camera|PopCameraMode)
      (Ability|RemoveGameplayEffectFromOwnerWithHandle (Variables|Default|GetTeleportingGEOnPlayer)))))

(event Custom|QuestUpdate
  (params (Pawn :type Fort%20Pawn%20Object%20Reference))
  ;[1 1 1] Hidden passage
  ;[1 1 1] Update Quests
  (Quest|SendComplexCustomStatEvent (Quests|GetQuestManager (Fortnite|GetFortPlayerControllerfromActor Pawn) "Invalid") 0 "(GameplayTags=((TagName=\"Athena.Quests.HidingProp.Teleport\")))" "(GameplayTags=)" 1))

(event Custom|OnPawnDied
  (params (DamagedActor :type Actor%20Object%20Reference) (Damage :type Float%20%28single-precision%29) (InstigatedBy :type Controller%20Object%20Reference) (DamageCauser :type Actor%20Object%20Reference) (HitLocation :type Vector) (FHitComponent :type Primitive%20Component%20Object%20Reference) (BoneName :type Name) (Momentum :type Vector))
  (bind _returnvalue (Ability|GetAvatarActorfromActorInfo))
  (bind _asfort_pawn (Utilities|Casting|CastToFortPawn _returnvalue))
  ;[1 1 1] Hidden passage
  ;[1 1 1] Cancel everything if player gets any of these tags
  (Class|BHidingProp|StopHiding (Variables|Default|GetTargetTeleporter) _asfort_pawn)
  (Class|BHidingProp|StopHiding (Variables|Default|GetHidingProp) _asfort_pawn)
  (Ability|EndAbility))

(event Custom|GrapplerFailSafe
  (params (FortPawn :type Fort%20Pawn%20Object%20Reference))
  (bind _returnvalue (Ability|GetAbilitySystemComponent FortPawn))
  ;[0.6158 0.3512 1] Grappler Failsafe for FORT-298499.  There's some replication breakdown keeping this from being solved via ability tags, so doing this instead
  (Utilities|IsValid _returnvalue
    (:Is%20Valid
      (bind _asfort_ability_system_component (Utilities|Casting|CastToFortAbilitySystemComponent _returnvalue))
      (Tags|UnBlockAbilitiesWithTags _asfort_ability_system_component (Variables|Default|GetChangeEquipmentTag) false))))
