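; Gameplay ability for the Papaya phone booth hiding prop, which lets a hidden player use a standalone Locker screen.
; On activation it holsters the weapon, then the client disables input and opens the Locker while the server applies an in-locker effect.
; WaitNetSync pairs client and server for exit, driving door, blinds, montage and delayed input return.
; Menu close, exit tags, storm, downed-type states or prop death end it, applying a cooldown.
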(event Custom|WaitForServerThenAnimate
  ;[0.1647 1 0] Run exit animation for client and enable input after delay
  (Ability|Tasks|WaitNetSync "BothWait"
    (:OnSync
      (Ability|Tasks|PlayMontageAndWait "None" (Variables|Default|GetShowOffMontage)) ; Will be cancelled by the server after we enabled input and move
      (Ability|Tasks|WaitDelay (Variables|Default|GetInputEnableExitDelay)
        (:OnFinish
          (bind _pawn (Variables|Default|GetPawn)
            (:then
              (Input|EnableInput _pawn (Fortnite|GetFortPlayerControllerfromActor (Ability|GetAvatarActorfromActorInfo))))
            (:else)))))))

(event Ability|EventFailedToActivatePassiveAbility
  (params (ActorInfo :type Gameplay%20Ability%20Actor%20Info%20Structure)))

(event Custom|OnLockerClosed
  (params (Widget :type Common%20Activatable%20Widget%20Object%20Reference) (Controller :type Player%20Controller%20Object%20Reference))
  ;[0.1647 1 0] Load the Locker UI and wait for player to close it
  (Ability|Tasks|WaitNetSync "BothWait"
    (:OnSync
      (CallFunction|WaitforServerThenAnimate))))

(event Custom|WaitForClientThenExit
  ;[1 1 1] Wait for the player to close the UI and exit the locker
  (Ability|Tasks|WaitNetSync "BothWait"
    (:OnSync
      (CallFunction|DebugAbilityLogging true false "GA_Athena_PapayaBooth Exit: ActivateAbilityWaithForClientThenExit.OnSync: ExitingLocker.True")
      (Variables|Default|SetbExitingLocker true)
      (Ability|RemoveGameplayEffectFromOwnerWithHandle (Variables|Default|GetHandle_InLocker))
      ;[1 1 1] Set Client Pawn Rotation so the Camera is oriented correctly to the booth on exit
      (bind _pawn (Variables|Default|GetPawn)
        (:then
          (Pawn|SetClientPawnNewRotation (Fortnite|GetFortPlayerControllerfromActor _pawn) (Math|Rotator|MakeRotator 0.0 0.0 (.yaw (Math|Rotator|MakeRotfromX (* (Variables|Default|GetHidingPropForward) -1.0))))))
        (:else
          _))
      (bind _hidingprop (Variables|Default|GetHidingProp)
        (:then
          (Networking|FlushNetDormancy _hidingprop))
        (:else
          _))
      (Ability|Tasks|WaitDelay 0.3
        (:OnFinish
          (bind _pawn_2 (Variables|Default|GetPawn)
            (:then
              (Ability|SendGameplayEventtoActor _pawn_2 "(TagName=\"Event.Athena.HidingProp.StopHiding\")" (Utilities|Struct|MakeGameplayEventData :Instigator (Variables|Default|GetHidingProp) :OptionalObject (Variables|Default|GetHidingProp) :ContextHandle "()" :InstigatorTags "(GameplayTags=)" :TargetTags "(GameplayTags=)" :EventMagnitude (Variables|Default|GetExitMagnitude) :TargetData "()"))
              (Class|FortPlayerPawn|PostResIn (Utilities|Casting|CastToFortPlayerPawn _pawn_2)))
            (:else
              _))
          (bind _hidingprop_2 (Variables|Default|GetHidingProp)
            (:then
              ;[1 1 1] Apply Door Animation
              (bind _asb_hiding_prop_papaya_booth (Utilities|Casting|CastToB_HidingProp_Papaya_Booth _hidingprop_2))
              (Class|BHidingPropPapayaBooth|DoorSwingAnim _asb_hiding_prop_papaya_booth 0)
              (if (< (Utilities|Array|Length (Variables|/Game/Athena/Items/EnvironmentalItems/HidingProps/B_HidingProp.B_HidingProp_C|GetHidingPlayers _asb_hiding_prop_papaya_booth)) 1)
                (Class|BHidingPropPapayaBooth|BlindsAnimExit _asb_hiding_prop_papaya_booth)
                (else
                  _))
              (Ability|Tasks|WaitNetSync "BothWait"
                (:OnSync
                  (Ability|Tasks|PlayMontageAndWait "None" (Variables|Default|GetShowOffMontage)
                    (:then
                      (Ability|Tasks|WaitDelay (Variables|Default|GetInputEnableExitDelay)
                        (:then
                          ;[1 1 1] Secondary knock back attempt for players exiting at the same time
                          (Ability|Tasks|WaitDelay 0.5
                            (:OnFinish
                              (bind _pawn_3 (Variables|Default|GetPawn)
                                (:then
                                  (bind _asfort_player_pawn (Utilities|Casting|CastToFortPlayerPawn _pawn_3))
                                  (bind _hidingprop_3 (Variables|Default|GetHidingProp)
                                    (:then
                                      (bind _asb_hiding_prop_papaya_booth_2 (Utilities|Casting|CastToB_HidingProp_Papaya_Booth _hidingprop_3))
                                      (Class|BHidingPropPapayaBooth|SecondaryExitKnockBack _asb_hiding_prop_papaya_booth_2 _asfort_player_pawn))
                                    (:else)))
                                (:else)))))
                        (:OnFinish
                          (Utilities|WaitforVelocityExceeded (Variables|Default|GetMinVelocityToCancel)
                            (:Exceeded
                              _)) ; Cancel Ability if you start moving before anim montage is over
                        )))
                    (:OnCompleted
                      _)
                    (:OnInterrupted
                      _)
                    (:OnCancelled
                      _))
                  (Ability|EndAbility))))
            (:else
              _)))))))

(event Ability|EventActivateAbility
  (bind _pawn (Variables|Default|GetPawn))
  (bind _returnvalue (Fortnite|GetFortPlayerControllerfromActor _pawn))
  (bind _standalonelockerscreenclass (Variables|FortGameplayAbility_AthenaPapayaBooth|GetStandaloneLockerScreenClass))
  (bind _returnvalue_2 (Utilities|IsValidSoftClassReference _standalonelockerscreenclass))
  (bind _hidingprop (Variables|Default|GetHidingProp))
  ;[1 1 1] Hide HUD, save off hiding props and player pawns
  (bind _asb_hiding_prop (Utilities|Casting|CastToB_HidingProp (Ability|EffectContext|GetEffectCauser (Ability|GetGrantedbyEffectContext)))
    (:then
      (Variables|Default|SetHidingProp _asb_hiding_prop)
      ;[1 1 1] Holster weapon to hide pickaxe. (when swapping character parts, player mesh stands up and ignores the animation, so weapon must be hidden manually so it doesn't stick out of the phonebooth)
      (bind _asfort_pawn (Utilities|Casting|CastToFortPawn (Ability|GetAvatarActorfromActorInfo))
        (:then
          (bind _output_get (Variables|Default|SetPawn _asfort_pawn))
          (Weapon|SetHolsterWeapon _output_get 0 true false (Variables|Default|GetHolsterId))
          (Networking|CanExecuteCosmeticEvents
            (:True
              ;[0.1647 1 0] Load the Locker UI and wait for player to close it
              (CallFunction|ListenforExitEvent)
              (CallFunction|ExitfromStorm)
              (Input|DisableInput (Variables|Default|GetPawn) (Fortnite|GetFortPlayerControllerfromActor (Ability|GetAvatarActorfromActorInfo)))
              (CallFunction|DebugAbilityLogging true false (Utilities|String|Append "GA_Athena_PapayaBooth.ActivateAbility. StandaloneLockerScreenClassSoftRef.IsValid?: 
" (Utilities|String|ToString(Boolean) _returnvalue_2) "
If invalid, ensure StandaloneLocker plugin is enabled."))
              (if (and _returnvalue_2 (Pawn|IsLocallyControlled _pawn))
                (bind _userwidget (Class|FortAsyncActionCreateWidgetAsync|CreateWidgetAsync _standalonelockerscreenclass _returnvalue)
                  (:OnComplete
                    (Utilities|FlowControl|Sequence
                      (:then_0
                        (FortUIManagement|PushContentWidgetforPlayer _returnvalue _userwidget))
                      (:then_1
                        (bind _asfort_standalone_frontend (Utilities|Casting|CastToFortStandaloneFrontend _userwidget)
                          (:then
                            (bind _output_get_2 (Variables|Default|SetFrontEndRef _asfort_standalone_frontend))
                            (Events|AssignOnMenuClosed _output_get_2 (delegate Custom|OnLockerClosed)))
                          (:CastFailed
                            _))))))
                (else
                  _))
              (Ability|Tasks|WaitNetSync "BothWait"
                (:OnSync
                  (CallFunction|WaitforServerThenAnimate))))
            (:False
              (HUD|SetHUDElementVisibility (BlueprintContext|GetFortHUDContext) (Variables|Default|GetT_HudElementToHide) true "None")
              (Variables|Default|SetHidingPropForward (Transformation|GetActorForwardVector _hidingprop))
              (Variables|Default|SetHidingPlayers (Variables|/Game/Athena/Items/EnvironmentalItems/HidingProps/B_HidingProp.B_HidingProp_C|GetHidingPlayers _hidingprop))
              (CallFunction|WaitforTags)
              (Utilities|IsValid (Utilities|Array|Get(acopy) (Variables|Default|GetHidingPlayers) (Utilities|Array|LastIndex (Variables|Default|GetHidingPlayers)))
                (:Is%20Valid
                  (bind _returnvalue_3 (Ability|ApplyGameplayEffectToOwner (Variables|Default|GetGE_InLocker)))
                  (Variables|Default|SetHandle_InLocker _returnvalue_3)
                  (CallFunction|WaitforClientThenExit))
                (:Is%20Not%20Valid
                  (CallFunction|WaitforClientThenExit))))))
        (:CastFailed
          _)))
    (:CastFailed
      _))
  (Networking|CanExecuteCosmeticEvents
    (:True
      (Ability|Tasks|WaitNetSync "BothWait"
        (:OnSync
          (CallFunction|WaitforServerThenAnimate))))
    (:False
      (CallFunction|WaitforClientThenExit))))

(event Custom|WaitForTags
  ;[1 1 1] Cancel everything if player gets any of these tags or the prop is destroyed.
  (Ability|Tasks|WaitGameplayTagAdd "(TagName=\"Gameplay.Status.DBNO\")"
    (:then
      (Ability|Tasks|WaitGameplayTagAdd "(TagName=\"Gameplay.Action.Player.CarryingPlayer\")"
        (:then
          (Ability|Tasks|WaitGameplayTagAdd "(TagName=\"Athena.Item.Snowman\")"
            (:then
              (bind (_asynctaskproxy _payload) (Ability|Tasks|WaitGameplayEvent "(TagName=\"Event.Athena.HidingProp.PropExited\")" 0 true)
                (:then
                  (EventDispatchers|BindEventtoOnDied (Variables|Default|GetHidingProp) (delegate Custom|End)))
                (:EventReceived
                  (CallFunction|DebugAbilityLogging true false "GA_Athena_PapayaBooth Exit: ActivateAbility.WaitForTags: Event.Athena.HidingProp.PropExited added
From: " _payload)
                  (if (not (Variables|Default|GetbExitingLocker))
                    (bind _hidingprop (Variables|Default|GetHidingProp)
                      (:then
                        (bind _asb_hiding_prop_papaya_booth (Utilities|Casting|CastToB_HidingProp_Papaya_Booth _hidingprop))
                        (Class|BHidingPropPapayaBooth|BlindsAnimExit _asb_hiding_prop_papaya_booth)
                        (Ability|EndAbility))
                      (:else))))))
            (:Added
              _)))
        (:Added
          _))
      (bind _hidingprop_2 (Variables|Default|GetHidingProp)
        (:then
          (Class|BHidingProp|StopHiding _hidingprop_2 (Variables|Default|GetPawn)))
        (:else
          _))
      (Ability|EndAbility))
    (:Added
      _)))

(event Custom|End
  (params (DamagedActor :type Actor%20Object%20Reference) (Damage :type Float%20%28single-precision%29) (InstigatedBy :type Controller%20Object%20Reference) (DamageCauser :type Actor%20Object%20Reference) (HitLocation :type Vector) (FHitComponent :type Primitive%20Component%20Object%20Reference) (BoneName :type Name) (Momentum :type Vector))
  ;[1 1 1] Cancel everything if player gets any of these tags or the prop is destroyed.
  (Ability|EndAbility))

(event Ability|EventOnEndAbility
  (params (bWasCancelled :type Boolean))
  ;[1 1 1] End Ability, Turn HUD back on. Enable input. UnHolster Weapon
  (Ability|ApplyGameplayEffectToOwner "/Game/Athena/Items/EnvironmentalItems/HidingProps/GE_Athena_HidingProp_PhantomBoothCoolDown.GE_Athena_HidingProp_PhantomBoothCoolDown_C")
  (Input|EnableInput (Ability|GetAvatarActorfromActorInfo) (Fortnite|GetFortPlayerControllerfromActor (Ability|GetAvatarActorfromActorInfo)))
  (HUD|SetHUDElementVisibility (BlueprintContext|GetFortHUDContext) (Variables|Default|GetT_HudElementToHide) false "None")
  (bind _pawn (Variables|Default|GetPawn)
    (:then
      (Weapon|SetHolsterWeapon _pawn 0 false false (Variables|Default|GetHolsterId)))
    (:else
      _))
  (Ability|RemoveGameplayEffectFromOwnerWithHandle (Variables|Default|GetHandle_InLocker))
  (CallFunction|CloseLockerUI))

(event Custom|CloseLockerUI
  ;[0 0 0 0.5] Close the LockerUI
  (bind _frontendref (Variables|Default|GetFrontEndRef)
    (:then
      (FortUIManagement|PopAllWidgetsUptoandIncluding)
      (Class|FortStandaloneFrontend|HandleCloseAction _frontendref)
      (Variables|Default|SetFrontEndRef 0) ; Null this out so we only call this once.
    )
    (:else)))

(event Custom|ExitFromStorm
  ;[0 0 0 0.5] If the HidingProp is set to block entrance when the player is in the storm,
  ; set it to exit the player if the storm passes while hiding.
  (if (CallFunction|BlocksEntranceinStorm)
    (Ability|Tasks|WaitGameplayTagAdd (Variables|Default|GetOutsideSafeZoneTag)
      (:Added
        (CallFunction|WaitforServerThenAnimate)
        (CallFunction|CloseLockerUI) ; Okay to close immediately while server\nand client sync up for the animation.
      ))))

(event Custom|ListenForExitEvent
  ;[1 1 1] Listen for event(s) that will close the UI and exit the player.
  (Ability|Tasks|WaitGameplayTagAdd "(TagName=\"Event.Athena.HidingProp.PapayaBooth.Exit\")"
    (:then
      ;;[0.15 0.02192 0 0.5] TODO: Future goal here is to deprecate this creative tag since it can use the same newer event tag. Need to check both can share the same execution paths first.
      ;;;[1 0.8099 0.07087 0.5] For Creative Device: Wait for the device to apply the tag that should kick players out of the Phone Booth on channel received.
      (Ability|Tasks|WaitGameplayTagAdd "(TagName=\"Event.Creative.Device.PhoneBooth.Exit\")"
        (:Added
          (CallFunction|WaitforClientThenExit)
          (CallFunction|WaitforServerThenAnimate))))
    (:Added
      (CallFunction|WaitforServerThenAnimate)
      (CallFunction|CloseLockerUI) ; Okay to close immediately while server\nand client sync up for the animation.
    )))

(fn BlocksEntranceInStorm
  (return (GameplayTags|HasTag (GameplayTags|MakeGameplayTagContainerfromArray (Variables|/Game/Athena/Items/EnvironmentalItems/HidingProps/B_HidingProp.B_HidingProp_C|GetBlockEntranceTags (Variables|Default|GetHidingProp))) (Variables|Default|GetOutsideSafeZoneTag) true)))

(fn DebugAbilityLogging
  (params (BuildLogEnabled? :type Boolean :default true) (PIE%20PrintEnabled? :type Boolean) (Message :type String) (GameplayEventData :type Gameplay%20Event%20Data%20Structure%20%28by%20ref%29))
  (bind _instigator (Utilities|Struct|BreakGameplayEventData GameplayEventData))
  (bind _returnvalue (Utilities|GetDisplayName _instigator))
  (bind _returnvalue_2 (Utilities|String|Append Message _returnvalue))
  (Utilities|FlowControl|Sequence
    (:then_0
      (if BuildLogEnabled?
        (Fortnite|FortShippingLog self _returnvalue_2)))
    (:then_1
      (if PIE%20PrintEnabled?
        (Development|PrintString _returnvalue_2)))))
