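; HidingPropInterface is a Blueprint interface for Athena environmental hiding props.
; It declares a GetCanRustleAndWobble query, apparently for whether a prop may rustle and wobble.
; The function body here only returns false.
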(fn GetCanRustleAndWobble
  (return false))
