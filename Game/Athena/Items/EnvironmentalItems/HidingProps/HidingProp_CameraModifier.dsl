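; Camera modifier for hiding props that keeps the view from dropping too low while it is active.
; BlueprintModifyCamera clamps the view location's Z to at least 55 units above the pawn's actor location, with an effectively unlimited upper bound.
; View X and Y, rotation and FOV pass through unchanged.
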(fn BlueprintModifyCamera
  (params (DeltaTime :type Float%20%28single-precision%29) (ViewLocation :type Vector :default "0, 0, 0") (ViewRotation :type Rotator :default "0, 0, 0") (FOV :type Float%20%28single-precision%29))
  (return (Math|Vector|MakeVector (.x ViewLocation) (.y ViewLocation) (Math|Float|Clamp(Float) (.z ViewLocation) (+ (.z (Transformation|GetActorLocation (Variables|Default|GetPawn))) 55.0) 1e+10)) ViewRotation FOV))
